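# Character material for the PsychicAphid male medium body cosmetic, wrapping the shared CharacterParent 2023 PreFX and PostFX functions.
# An optional anodized effect adds Fresnel-tinted color and a light-reflection highlight to base color.
# Optional animated emissive uses three polar UV sets from pre-skinned position, feeding MF_PsychicChestEmissive with radial falloffs and RadialArea mask channels.
# Each set blends two half-offset samples by sine, apparently to hide the angular seam.
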Material /BRCosmetics/Characters/Player/Male/Medium/Bodies/M_MED_PsychicAphid/Materials/M_PsychicAphid
graph {
 properties {
  bUseMaterialAttributes=true
  bUsedWithMeshDeformer=true
  bUsedWithMorphTargets=true
  bUsedWithSkeletalMesh=true
  pixelDepthOffsetMode=PDOM_Legacy
  shadingModel=MSM_FromMaterialExpression
  usedShadingModels=Default Lit | Subsurface | Clear Coat | Subsurface Profile | Cloth
 }
 nodes [
  n1 = MaterialFunctionCall() {materialFunction=&MF_CharacterParent_2023_PreFX}
  MaterialFunctionCall(Material=NamedRerouteUsage(), M=@n1.M, Pre Skinned Local Position=@n1.M, Distance Blend=@n1.M, Fresnel=@n1.M) {materialFunction=&MF_CharacterParent_2023_PostFX}
  NamedRerouteDeclaration(None=@n1.Result) {name=MA_IN, nodeColor=<rgba 0.1953 0 1 1>}
  n2 = GetMaterialAttributes(None=NamedRerouteUsage()) {attributeGetTypes [#1]}
  n3 = Constant() {r=1}
  n4 = Constant() {r=3}
  n5 = PixelNormalWS()
  n6 = Multiply(
   A=TextureSample(UVs=Fresnel(ExponentIn=@n3, BaseReflectFractionIn=@n4, Normal=@n5) {baseReflectFraction=0.5, exponent=3}) {texture=/Game/Global/Textures/Default/Masks_and_Gradients/3pxRGBImage}.RGB
   B=VectorParameter() {group=TowerSentinel_Metal, parameterName=AnodizationFilter, defaultValue=<rgba 1 0 1 0>}.RGB
  )
  n7 = Constant() {r=4}
  n8 = AtmosphericLightVector()
  n9 = ReflectionVectorWS()
  n10 = Multiply(
   A=Power(Base=DotProduct(A=@n8, B=@n9), Exp=ScalarParameter() {defaultValue=55, group=TowerSentinel_Metal, parameterName=Anodized_pow})
   B=Constant() {r=512}
  )
  n11 = Multiply(
   A=LinearInterpolate(A=@n6, B=Multiply(A=@n7, B=@n6), Alpha=Saturate(None=@n10))
   B=LinearInterpolate(
    A=ScalarParameter() {group=TowerSentinel_Metal, parameterName=AnodizedEffectIntensity}
    B=ScalarParameter() {group=TowerSentinel_Metal, parameterName=AnodizedEffectIntensity2}
    Alpha=NamedRerouteUsage()
   )
  ) {constB=0.1}
  n12 = Constant()
  n13 = StaticSwitchParameter(True=ShadingPathSwitch(Default=Multiply(A=@n11), Forward=@n12, Mobile=@n12), False=@n12) {group=TowerSentinel, parameterName=UseAnodizedEffect}
  n14 = GetMaterialAttributes(None=SetMaterialAttributes!(MaterialAttributes=@n2.MaterialAttributes, Base Color=Add(A=@n2.Base Color, B=@n13)) {attributeSetTypes [#1]}) {attributeGetTypes [#2]}
  n15 = NamedRerouteUsage()
  n16 = NamedRerouteUsage()
  n17 = NamedRerouteUsage()
  n18 = LinearInterpolate(A=@n14.Emissive Color, B=@n15, Alpha=NamedRerouteUsage())
  n19 = StaticSwitchParameter(True=LinearInterpolate(A=@n18, B=@n16, Alpha=NamedRerouteUsage()), False=@n18) {defaultValue=true, group=TowerSentinel_RadialUV, parameterName=UseEM02}
  n20 = NamedRerouteUsage()
  n21 = StaticSwitchParameter(
   True=Add(A=Add(A=@n15, B=@n16), B=@n17)
   False=StaticSwitchParameter(True=LinearInterpolate(A=@n19, B=@n17, Alpha=@n20), False=@n19) {defaultValue=true, group=TowerSentinel_RadialUV, parameterName=UseEM03}
  ) {group=TowerSentinel_RadialUV, parameterName=DebugRadialEffect}
  NamedRerouteDeclaration(None=SetMaterialAttributes!(
    MaterialAttributes=@n14.MaterialAttributes
    Emissive Color=StaticSwitchParameter(True=@n21, False=@n14.Emissive Color) {group=TowerSentinel_RadialUV, parameterName=UseRadialEffect}
   ) {attributeSetTypes [#2]}) {name=MA_OUT, nodeColor=<rgba 0.125 1 0 1>}
  n22 = NamedRerouteUsage()
  n23 = NamedRerouteUsage()
  n24 = Subtract(
   A=VectorParameter() {group=TowerSentinel_RadialUV, parameterName=RadialUVOffset02, defaultValue=<rgba 0 0 130 0>}.RGB
   B=VertexInterpolator(VS=PreSkinnedPosition()).PS
  )
  n25 = MaterialFunctionCall(Float2=MaterialFunctionCall(
    UVs=ComponentMask(None=@n24) {b=true, r=true}
    Texture Scale=VectorParameter() {group=TowerSentinel_RadialUV, parameterName=RadialUVScale02, defaultValue=<rgba 1 1 1 0>}.RGB
   ) {materialFunction=&ScaleUVsByCenter}.UVs) {materialFunction=&BreakOutFloat2Components}
  n26 = MaterialFunctionCall(X=@n25.R, Y=ComponentMask(None=@n24) {g=true}, Z=@n25.R) {materialFunction=&MakeFloat3}
  n27 = StaticComponentMaskParameter(None=@n26.Result) {defaultB=true, defaultR=true, group=TowerSentinel_RadialUV, parameterName=Area02Channel}
  n28 = Add(
   A=@n27
   B=ComponentMask(None=VectorParameter() {group=TowerSentinel, parameterName=maskOffset02, defaultValue=<rgba 0 -5 0 0>}.RGB) {g=true, r=true}
  )
  n29 = ComponentMask(None=VectorParameter() {group=TowerSentinel, parameterName=maskScale02, defaultValue=<rgba 2.2 1 0 0>}.RGB) {g=true, r=true}
  n30 = Divide(
   A=Distance(A=Multiply(A=@n28, B=@n29), B=Constant())
   B=StaticSwitchParameter(True=ScalarParameter() {defaultValue=100, group=TowerSentinel_RadialUV, parameterName=StaticRadius}, False=ObjectRadius()) {group=TowerSentinel_RadialUV, parameterName=UseStaticRadius}
  )
  n31 = Multiply(
   A=Power(Base=OneMinus(None=@n30), Exp=ScalarParameter() {defaultValue=35, group=TowerSentinel, parameterName=maskPow02})
   B=ScalarParameter() {defaultValue=6, parameterName=maskMul02}
  )
  n32 = ComponentMask(None=NamedRerouteUsage()) {r=true}
  n33 = Multiply(A=Power(Base=Sine(None=@n32) {period=2})) {constB=5}
  n34 = Add(A=NamedRerouteUsage(), B=Constant2Vector() {r=0.5})
  n35 = LinearInterpolate(
   A=MaterialFunctionCall!(
    In=Frac(None=@n34)
    Speed02=NamedRerouteUsage()
    Speed03=NamedRerouteUsage()
    Speed04=NamedRerouteUsage()
    Speed01=NamedRerouteUsage()
    SpeedMain_02=NamedRerouteUsage()
    SpeedMain_01=NamedRerouteUsage()
   ) {materialFunction=&/MF_PsychicChestEmissive}.Result
   B=MaterialFunctionCall(
    In=NamedRerouteUsage()
    Speed02=NamedRerouteUsage()
    Speed03=NamedRerouteUsage()
    Speed04=NamedRerouteUsage()
    Speed01=NamedRerouteUsage()
    SpeedMain_02=NamedRerouteUsage()
    SpeedMain_01=NamedRerouteUsage()
   ) {materialFunction=&MF_PsychicChestEmissive}.Result
   Alpha=Saturate(None=Saturate(None=@n33))
  )
  NamedRerouteDeclaration(None=Add(A=Multiply(A=@n22, B=@n23) {constB=0.3}, B=Multiply(A=@n31, B=@n35))) {name=ChestColor02, nodeColor=<rgba 0 0.0156 1 1>}
  n36 = Subtract(
   A=VectorParameter() {group=TowerSentinel_RadialUV, parameterName=RadialUVOffset03, defaultValue=<rgba 0 0 130 0>}.RGB
   B=VertexInterpolator(VS=PreSkinnedPosition()).PS
  )
  n37 = MaterialFunctionCall(Float2=MaterialFunctionCall(
    UVs=ComponentMask(None=@n36) {b=true, r=true}
    Texture Scale=VectorParameter() {group=TowerSentinel_RadialUV, parameterName=RadialUVScale03, defaultValue=<rgba 1 1 1 0>}.RGB
   ) {materialFunction=&ScaleUVsByCenter}.UVs) {materialFunction=&BreakOutFloat2Components}
  n38 = MaterialFunctionCall(X=@n37.R, Y=ComponentMask(None=@n36) {g=true}, Z=@n37.R) {materialFunction=&MakeFloat3}
  n39 = StaticComponentMaskParameter(None=Divide(A=@n38.Result, B=PreSkinnedLocalBounds().Extents)) {defaultB=true, defaultR=true, group=TowerSentinel_RadialUV, parameterName=Area03Channel}
  n40 = Divide(A=Arctangent2Fast(Y=ComponentMask(None=@n39) {g=true}, X=ComponentMask(None=@n39) {r=true}), B=Constant() {r=3.1416})
  n41 = StaticComponentMaskParameter(None=@n38.Result) {defaultB=true, defaultR=true, group=TowerSentinel_RadialUV, parameterName=Area03Channel}
  n42 = Distance(A=@n41, B=Constant())
  n43 = StaticSwitchParameter(True=ScalarParameter() {defaultValue=100, group=TowerSentinel_RadialUV, parameterName=StaticRadius}, False=ObjectRadius()) {group=TowerSentinel_RadialUV, parameterName=UseStaticRadius}
  n44 = AppendVector(A=Multiply(A=Add(A=@n40) {constB=-0.5}) {constB=0.5}, B=OneMinus(None=Divide(A=@n42, B=@n43)))
  n45 = ComponentMask(None=VectorParameter() {group=TowerSentinel_RadialUV, parameterName=UV_Mul03, defaultValue=<rgba 1 1 0 0>}.RGB) {g=true, r=true}
  NamedRerouteDeclaration(None=Frac(None=Multiply(A=@n44, B=@n45))) {name=RadialUV_2, nodeColor=<rgba 1 0.2812 0 1>}
  n46 = NamedRerouteUsage()
  n47 = NamedRerouteUsage()
  n48 = Add(
   A=@n41
   B=ComponentMask(None=VectorParameter() {group=TowerSentinel, parameterName=maskOffset03, defaultValue=<rgba 0 -5 0 0>}.RGB) {g=true, r=true}
  )
  n49 = ComponentMask(None=VectorParameter() {group=TowerSentinel, parameterName=maskScale03, defaultValue=<rgba 2.2 1 0 0>}.RGB) {g=true, r=true}
  n50 = Divide(
   A=Distance(A=Multiply(A=@n48, B=@n49), B=Constant())
   B=StaticSwitchParameter(True=ScalarParameter() {defaultValue=100, group=TowerSentinel_RadialUV, parameterName=StaticRadius}, False=ObjectRadius()) {group=TowerSentinel_RadialUV, parameterName=UseStaticRadius}
  )
  n51 = Multiply(
   A=Power(Base=OneMinus(None=@n50), Exp=ScalarParameter() {defaultValue=35, group=TowerSentinel, parameterName=maskPow03})
   B=ScalarParameter() {defaultValue=6, parameterName=maskMul03}
  )
  n52 = Add(A=NamedRerouteUsage(), B=Constant2Vector() {r=0.5})
  n53 = ComponentMask(None=NamedRerouteUsage()) {r=true}
  n54 = Multiply(A=Power(Base=Sine(None=@n53) {period=2})) {constB=5}
  n55 = LinearInterpolate(
   A=MaterialFunctionCall(
    In=Frac(None=@n52)
    Speed02=NamedRerouteUsage()
    Speed03=NamedRerouteUsage()
    Speed04=NamedRerouteUsage()
    Speed01=NamedRerouteUsage()
    SpeedMain_02=NamedRerouteUsage()
    SpeedMain_01=NamedRerouteUsage()
   ) {materialFunction=&MF_PsychicChestEmissive}.Result
   B=MaterialFunctionCall(
    In=NamedRerouteUsage()
    Speed02=NamedRerouteUsage()
    Speed03=NamedRerouteUsage()
    Speed04=NamedRerouteUsage()
    Speed01=NamedRerouteUsage()
    SpeedMain_02=NamedRerouteUsage()
    SpeedMain_01=NamedRerouteUsage()
   ) {materialFunction=&MF_PsychicChestEmissive}.Result
   Alpha=Saturate(None=Saturate(None=@n54))
  )
  NamedRerouteDeclaration(None=Add(A=Multiply(A=@n46, B=@n47) {constB=0.3}, B=Multiply(A=@n51, B=@n55))) {name=ChestColor03, nodeColor=<rgba 0 0.0156 1 1>}
  NamedRerouteDeclaration(None=@n14.Emissive Color) {name=EM_IN, nodeColor=<rgba 0 1 0.4844 1>}
  n56 = VectorParameter() {
   group=TowerSentinel_RadialUV
   parameterName=SpeedControl
   channelNames {a=SpeedY02, b=SpeedX02, g=SpeedY01, r=SpeedX01}
   defaultValue=<rgba 0 0.55 0 0.35>
  }
  NamedRerouteDeclaration(None=AppendVector(A=@n56.SpeedX01, B=@n56.SpeedX01)) {name=MainSpeed01, nodeColor=<rgba 1 0 0.6562 1>}
  NamedRerouteDeclaration(None=AppendVector(A=@n56.SpeedX02, B=@n56.SpeedX02)) {name=MainSpeed02, nodeColor=<rgba 0 0.9766 1 1>}
  n57 = VectorParameter() {
   group=TowerSentinel_RadialUV
   parameterName=SpeedControl01
   channelNames {a=SpeedY02, b=SpeedX02, g=SpeedY01, r=SpeedX01}
   defaultValue=<rgba 0 0.35 0 0.1>
  }
  NamedRerouteDeclaration(None=AppendVector(A=@n57.SpeedX01, B=@n57.SpeedX01)) {name=Speed01, nodeColor=<rgba 1 0 0.6562 1>}
  NamedRerouteDeclaration(None=AppendVector(A=@n57.SpeedX02, B=@n57.SpeedX02)) {name=Speed02, nodeColor=<rgba 0 0.9766 1 1>}
  n58 = VectorParameter() {
   group=TowerSentinel_RadialUV
   parameterName=SpeedControl02
   channelNames {a=SpeedY02, b=SpeedX02, g=SpeedY01, r=SpeedX01}
   defaultValue=<rgba 0.03 0.2 0.1 0.2>
  }
  NamedRerouteDeclaration(None=AppendVector(A=@n58.SpeedX01, B=@n58.SpeedX01)) {name=Speed03, nodeColor=<rgba 1 0 0.6562 1>}
  n59 = ChannelMaskParameter(None=@n1.Tech Art Mask) {group=TowerSentinel_Metal, maskChannel=Blue, parameterName=UseTowerSentinelChannel, defaultValue=<rgba 0 0 1 0>}
  NamedRerouteDeclaration(None=@n59.R) {name=TechArtMask, nodeColor=<rgba 0.9922 0 1 1>}
  NamedRerouteDeclaration(None=AppendVector(A=@n58.SpeedX02, B=@n58.SpeedX02)) {name=Speed04, nodeColor=<rgba 0 0.9766 1 1>}
  n60 = Subtract(
   A=VectorParameter() {group=TowerSentinel_RadialUV, parameterName=RadialUVOffset01, defaultValue=<rgba 0 0 130 0>}.RGB
   B=VertexInterpolator(VS=PreSkinnedPosition()).PS
  )
  n61 = MaterialFunctionCall(Float2=MaterialFunctionCall(
    UVs=ComponentMask(None=@n60) {b=true, r=true}
    Texture Scale=VectorParameter() {group=TowerSentinel_RadialUV, parameterName=RadialUVScale01, defaultValue=<rgba 1 1 1 0>}.RGB
   ) {materialFunction=&ScaleUVsByCenter}.UVs) {materialFunction=&BreakOutFloat2Components}
  n62 = MaterialFunctionCall(X=@n61.R, Y=ComponentMask(None=@n60) {g=true}, Z=@n61.R) {materialFunction=&MakeFloat3}
  n63 = StaticComponentMaskParameter(None=Divide(A=@n62.Result, B=PreSkinnedLocalBounds().Extents)) {defaultB=true, defaultR=true, group=TowerSentinel_RadialUV, parameterName=Area01Channel}
  n64 = Divide(A=Arctangent2Fast(Y=ComponentMask(None=@n63) {g=true}, X=ComponentMask(None=@n63) {r=true}), B=Constant() {r=3.1416})
  n65 = StaticComponentMaskParameter(None=@n62.Result) {defaultB=true, defaultR=true, group=TowerSentinel_RadialUV, parameterName=Area01Channel}
  n66 = Distance(A=@n65, B=Constant())
  n67 = StaticSwitchParameter(True=ScalarParameter() {defaultValue=100, group=TowerSentinel_RadialUV, parameterName=StaticRadius}, False=ObjectRadius()) {group=TowerSentinel_RadialUV, parameterName=UseStaticRadius}
  n68 = AppendVector(A=Multiply(A=Add(A=@n64) {constB=-0.5}) {constB=0.5}, B=OneMinus(None=Divide(A=@n66, B=@n67)))
  n69 = ComponentMask(None=VectorParameter() {group=TowerSentinel_RadialUV, parameterName=UV_Mul01, defaultValue=<rgba 1 1 0 0>}.RGB) {g=true, r=true}
  NamedRerouteDeclaration(None=Frac(None=Multiply(A=@n68, B=@n69))) {name=RadialUV, nodeColor=<rgba 1 0.2812 0 1>}
  n70 = NamedRerouteUsage()
  n71 = NamedRerouteDeclaration(None=ScalarParameter() {defaultValue=0.2, group=TowerSentinel_RadialUV, parameterName=BaseEM_Blend}) {name=EM_Blend, nodeColor=<rgba 1 0 0.4219 1>}
  n72 = Add(
   A=@n65
   B=ComponentMask(None=VectorParameter() {group=TowerSentinel, parameterName=maskOffset01, defaultValue=<rgba 0 -5 0 0>}.RGB) {g=true, r=true}
  )
  n73 = ComponentMask(None=VectorParameter() {group=TowerSentinel, parameterName=maskScale01, defaultValue=<rgba 2.2 1 0 0>}.RGB) {g=true, r=true}
  n74 = Divide(
   A=Distance(A=Multiply(A=@n72, B=@n73), B=Constant())
   B=StaticSwitchParameter(True=ScalarParameter() {defaultValue=100, group=TowerSentinel_RadialUV, parameterName=StaticRadius}, False=ObjectRadius()) {group=TowerSentinel_RadialUV, parameterName=UseStaticRadius}
  )
  n75 = Multiply(
   A=Power(Base=OneMinus(None=@n74), Exp=ScalarParameter() {defaultValue=35, group=TowerSentinel, parameterName=maskPow01})
   B=ScalarParameter() {defaultValue=6, parameterName=maskMul01}
  )
  n76 = Add(A=NamedRerouteUsage(), B=Constant2Vector() {r=0.5})
  n77 = ComponentMask(None=NamedRerouteUsage()) {r=true}
  n78 = Multiply(A=Power(Base=Sine(None=@n77) {period=2})) {constB=5}
  n79 = LinearInterpolate(
   A=MaterialFunctionCall(
    In=Frac(None=@n76)
    Speed02=NamedRerouteUsage()
    Speed03=NamedRerouteUsage()
    Speed04=NamedRerouteUsage()
    Speed01=NamedRerouteUsage()
    SpeedMain_02=NamedRerouteUsage()
    SpeedMain_01=NamedRerouteUsage()
   ) {materialFunction=&MF_PsychicChestEmissive}.Result
   B=MaterialFunctionCall(
    In=NamedRerouteUsage()
    Speed02=NamedRerouteUsage()
    Speed03=NamedRerouteUsage()
    Speed04=NamedRerouteUsage()
    Speed01=NamedRerouteUsage()
    SpeedMain_02=NamedRerouteUsage()
    SpeedMain_01=NamedRerouteUsage()
   ) {materialFunction=&MF_PsychicChestEmissive}.Result
   Alpha=Saturate(None=Saturate(None=@n78))
  )
  NamedRerouteDeclaration(None=Add(A=Multiply(A=@n70, B=@n71) {constB=0.3}, B=Multiply(A=@n75, B=@n79))) {name=ChestColor01, nodeColor=<rgba 0 0.0156 1 1>}
  n80 = TextureSampleParameter2D() {
   group=TowerSentinel_RadialUV
   parameterName=RadialArea
   samplerType=SAMPLERTYPE_Masks
   texture=&T_BlackMask
  }
  NamedRerouteDeclaration(None=ComponentMask(None=@n80.RGB) {r=true}) {name=Area01, nodeColor=<rgba 1 0.9141 0 1>}
  NamedRerouteDeclaration(None=ComponentMask(None=@n80.RGB) {g=true}) {name=Area02, nodeColor=<rgba 1 0.9141 0 1>}
  NamedRerouteDeclaration(None=ComponentMask(None=@n80.RGB) {b=true}) {name=Area03, nodeColor=<rgba 1 0.9141 0 1>}
  n81 = StaticComponentMaskParameter(None=Divide(A=@n26.Result, B=PreSkinnedLocalBounds().Extents)) {defaultB=true, defaultR=true, group=TowerSentinel_RadialUV, parameterName=Area02Channel}
  n82 = Divide(A=Arctangent2Fast(Y=ComponentMask(None=@n81) {g=true}, X=ComponentMask(None=@n81) {r=true}), B=Constant() {r=3.1416})
  n83 = Distance(A=@n27, B=Constant())
  n84 = StaticSwitchParameter(True=ScalarParameter() {defaultValue=100, group=TowerSentinel_RadialUV, parameterName=StaticRadius}, False=ObjectRadius()) {group=TowerSentinel_RadialUV, parameterName=UseStaticRadius}
  n85 = AppendVector(A=Multiply(A=Add(A=@n82) {constB=-0.5}) {constB=0.5}, B=OneMinus(None=Divide(A=@n83, B=@n84)))
  n86 = ComponentMask(None=VectorParameter() {group=TowerSentinel_RadialUV, parameterName=UV_Mul02, defaultValue=<rgba 1 1 0 0>}.RGB) {g=true, r=true}
  NamedRerouteDeclaration(None=Frac(None=Multiply(A=@n85, B=@n86))) {name=RadialUV_1, nodeColor=<rgba 1 0.2812 0 1>}
  SmoothStep(Value=@n80.RGB) {constMax=0.5}
 ]
}
