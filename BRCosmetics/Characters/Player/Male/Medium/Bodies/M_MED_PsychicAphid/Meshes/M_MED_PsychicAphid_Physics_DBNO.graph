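# Physics asset for the M_MED_PsychicAphid character body's DBNO (presumably down-but-not-out) variant, covering its cloak and lower body.
# Capsules on pelvis, legs and feet accompany five cloak chains (back, both fronts, both sides), each rooted in a small base capsule free-jointed to the pelvis.
# Chain joints allow no twist and limited swing, generally tighter toward the hem.
# Free constraints also pair the lower right front and side segments.
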PhysicsAsset /BRCosmetics/Characters/Player/Male/Medium/Bodies/M_MED_PsychicAphid/Meshes/M_MED_PsychicAphid_Physics_DBNO
bodies {
 Base_cloak_back [{shapeType=Capsule, length=0.505, radius=0.505, shapeName=Base_cloak_back_capsule, rotation=<rot 0 -90 0>}]
 Base_cloak_front_L [{shapeType=Capsule, length=1, radius=1, shapeName=Base_cloak_front_L_capsule}]
 Base_cloak_front_R [{shapeType=Capsule, length=0.505, radius=0.505, shapeName=Base_cloak_front_R_capsule, rotation=<rot 0 -90 0>}]
 Base_cloak_side_L [{shapeType=Capsule, length=0.505, radius=0.505, shapeName=Base_cloak_side_L_capsule, rotation=<rot 0 -90 0>}]
 Base_cloak_side_R [{shapeType=Capsule, length=0.505, radius=0.505, shapeName=Base_cloak_side_R_capsule, rotation=<rot 0 -90 0>}]
 Dyn_cloak_back_01 [{shapeType=Capsule, length=8.3404, radius=7.6078, shapeName=Dyn_cloak_back_01_capsule, center=<3.8415 0 4.6473>, rotation=<rot 0 0 90>}]
 Dyn_cloak_back_02 [{shapeType=Capsule, length=8.4339, radius=8.2143, shapeName=Dyn_cloak_back_02_capsule, center=<4.1622 0 5.6179>, rotation=<rot 0 0 90>}]
 Dyn_cloak_back_03 [{shapeType=Capsule, length=10.185, radius=8.2848, shapeName=Dyn_cloak_back_03_capsule, center=<3.1227 0 5.687>, rotation=<rot 0 0 90>}]
 Dyn_cloak_back_04 [{shapeType=Capsule, length=7.5419, radius=9.4, shapeName=Dyn_cloak_back_04_capsule, center=<7.38 0 4.7979>, rotation=<rot 0 0 90>}]
 Dyn_cloak_front_L_01 [{
   shapeType=Capsule
   length=1.1813
   radius=8.8187
   shapeName=Dyn_cloak_front_L_01_capsule
   center=<6.5884 -6.8526 5.2743>
   rotation=<rot 0 0.0002 75>
  }]
 Dyn_cloak_front_L_02 [{
   shapeType=Capsule
   length=0.9186
   radius=9.0814
   shapeName=Dyn_cloak_front_L_02_capsule
   center=<8.0804 -6.6133 4.3812>
   rotation=<rot 0 0.0002 75>
  }]
 Dyn_cloak_front_L_03 [{
   shapeType=Capsule
   length=2.6941
   radius=7.3059
   shapeName=Dyn_cloak_front_L_03_capsule
   center=<5.3191 -6.4913 1.8221>
   rotation=<rot 0 0.0002 59.9999>
  }]
 Dyn_cloak_front_L_04 [{
   shapeType=Capsule
   length=3.508
   radius=6.492
   shapeName=Dyn_cloak_front_L_04_capsule
   center=<5.3191 -5.8979 0>
   rotation=<rot 0 0.0003 59.9999>
  }]
 Dyn_cloak_front_R_01 [{
   shapeType=Capsule
   length=7.9864
   radius=8.1873
   shapeName=Dyn_cloak_front_R_01_capsule
   center=<6.7373 3.8075 6.4439>
   rotation=<rot 0 -0.0001 105>
  }]
 Dyn_cloak_front_R_02 [{
   shapeType=Capsule
   length=6.3498
   radius=9.8239
   shapeName=Dyn_cloak_front_R_02_capsule
   center=<6.7373 1.1982 7.5114>
   rotation=<rot 0 -0.0002 110>
  }]
 Dyn_cloak_front_R_03 [{
   shapeType=Capsule
   length=6.3498
   radius=9.8239
   shapeName=Dyn_cloak_front_R_03_capsule
   center=<8.5571 2.5346 5.4355>
   rotation=<rot 0 -0.0001 115>
  }]
 Dyn_cloak_front_R_04 [{shapeType=Capsule, length=8.6507, radius=9.8239, shapeName=Dyn_cloak_front_R_04_capsule, center=<8.7552 0 5.373>, rotation=<rot 0 0 90>}]
 Dyn_cloak_side_L_01 [{shapeType=Capsule, length=7.0971, radius=9.0766, shapeName=Dyn_cloak_side_L_01_capsule, center=<5.978 0 6.06>, rotation=<rot 0 0 90>}]
 Dyn_cloak_side_L_02 [{shapeType=Capsule, length=6.9761, radius=9.1976, shapeName=Dyn_cloak_side_L_02_capsule, center=<6.7373 0 5.9375>, rotation=<rot 0 0 90>}]
 Dyn_cloak_side_L_03 [{shapeType=Capsule, length=7.1866, radius=8.987, shapeName=Dyn_cloak_side_L_03_capsule, center=<8.3867 0 6.2044>, rotation=<rot 0 0 90>}]
 Dyn_cloak_side_R_01 [{shapeType=Capsule, length=7.6039, radius=8.5697, shapeName=Dyn_cloak_side_R_01_capsule, center=<6.7373 0 5.7498>, rotation=<rot 0 0 90>}]
 Dyn_cloak_side_R_02 [{shapeType=Capsule, length=6.6708, radius=9.5029, shapeName=Dyn_cloak_side_R_02_capsule, center=<6.7373 0 5.8559>, rotation=<rot 0 0 90>}]
 Dyn_cloak_side_R_03 [{shapeType=Capsule, length=6.3498, radius=9.8239, shapeName=Dyn_cloak_side_R_03_capsule, center=<6.7373 0 7.1662>, rotation=<rot 0 0 90>}]
 Dyn_cloak_side_R_04 [{
   shapeType=Capsule
   length=10.4019
   radius=10.2639
   shapeName=Dyn_cloak_side_R_04_capsule
   center=<6.7373 -0.1757 6.9027>
   rotation=<rot -0.0002 -0.0001 83.4494>
  }]
 calf_l [{shapeType=Capsule, length=27.1293, radius=8.5519, center=<-23.8348 -1.9368 0>, rotation=<rot 90 0 0>}]
 calf_r [{shapeType=Capsule, length=26.5638, radius=9.1173, center=<23.8348 1.9368 0>, rotation=<rot 90 0 180>}]
 foot_l [{shapeType=Capsule, length=13.7318, radius=9.8555, shapeName=foot_l_capsule, center=<-4.6806 7.0814 0>, rotation=<rot -0.0001 4.9993 -90>}]
 foot_r [{shapeType=Capsule, length=14.927, radius=11.1259, shapeName=foot_l_capsule, center=<4.6806 -7.0814 0>, rotation=<rot -0.0001 4.9993 90>}]
 pelvis [{shapeType=Capsule, length=3.545, radius=13.4307}]
 thigh_l [{shapeType=Capsule, length=20.8195, radius=9.1382, center=<-16.6963 0 0.718>, rotation=<rot 90 0 0>}]
 thigh_r [{shapeType=Capsule, length=21.8341, radius=9.9283, center=<17.5694 0 0>, rotation=<rot 90 0 0>}]
}
constraints [
 {bone1Name=thigh_l, bone2Name=pelvis, swing1LimitDegrees=45, swing2LimitDegrees=45, twistLimitDegrees=45}
 {bone1Name=thigh_r, bone2Name=pelvis, swing1LimitDegrees=45, swing2LimitDegrees=45, twistLimitDegrees=45}
 {
  bone1Name=foot_l
  bone2Name=calf_l
  swing1LimitDegrees=45
  swing1Motion=Free
  swing2LimitDegrees=45
  swing2Motion=Free
  twistLimitDegrees=45
  twistMotion=Free
 }
 {
  bone1Name=foot_r
  bone2Name=calf_r
  swing1LimitDegrees=45
  swing1Motion=Free
  swing2LimitDegrees=45
  swing2Motion=Free
  twistLimitDegrees=45
  twistMotion=Free
 }
 {
  bone1Name=Base_cloak_front_R
  bone2Name=pelvis
  swing1LimitDegrees=45
  swing1Motion=Free
  swing2LimitDegrees=45
  swing2Motion=Free
  twistLimitDegrees=45
  twistMotion=Free
 }
 {bone1Name=Dyn_cloak_front_R_01, bone2Name=Base_cloak_front_R, swing1LimitDegrees=18, swing2LimitDegrees=20, twistLimitDegrees=0}
 {bone1Name=Dyn_cloak_front_R_02, bone2Name=Dyn_cloak_front_R_01, swing1LimitDegrees=15, swing2LimitDegrees=18, twistLimitDegrees=0}
 {bone1Name=Dyn_cloak_front_R_03, bone2Name=Dyn_cloak_front_R_02, swing1LimitDegrees=10, swing2LimitDegrees=15, twistLimitDegrees=0}
 {bone1Name=Dyn_cloak_front_R_04, bone2Name=Dyn_cloak_front_R_03, swing1LimitDegrees=6, swing2LimitDegrees=10, twistLimitDegrees=0}
 {
  bone1Name=Base_cloak_side_L
  bone2Name=pelvis
  swing1LimitDegrees=6
  swing1Motion=Free
  swing2LimitDegrees=20
  swing2Motion=Free
  twistLimitDegrees=45
  twistMotion=Free
 }
 {bone1Name=Dyn_cloak_side_L_01, bone2Name=Base_cloak_side_L, swing1LimitDegrees=15, swing2LimitDegrees=30, twistLimitDegrees=0}
 {bone1Name=Dyn_cloak_side_L_02, bone2Name=Dyn_cloak_side_L_01, swing1LimitDegrees=10, swing2LimitDegrees=25, twistLimitDegrees=0}
 {bone1Name=Dyn_cloak_side_L_03, bone2Name=Dyn_cloak_side_L_02, swing1LimitDegrees=20, swing2LimitDegrees=25, twistLimitDegrees=0}
 {
  bone1Name=Base_cloak_side_R
  bone2Name=pelvis
  swing1LimitDegrees=45
  swing1Motion=Free
  swing2LimitDegrees=45
  swing2Motion=Free
  twistLimitDegrees=45
  twistMotion=Free
 }
 {bone1Name=Dyn_cloak_side_R_01, bone2Name=Base_cloak_side_R, swing1LimitDegrees=6, swing2LimitDegrees=25, twistLimitDegrees=0}
 {bone1Name=Dyn_cloak_side_R_02, bone2Name=Dyn_cloak_side_R_01, swing1LimitDegrees=5, swing2LimitDegrees=20, twistLimitDegrees=0}
 {bone1Name=Dyn_cloak_side_R_03, bone2Name=Dyn_cloak_side_R_02, swing1LimitDegrees=4, swing2LimitDegrees=15, twistLimitDegrees=0}
 {bone1Name=Dyn_cloak_side_R_04, bone2Name=Dyn_cloak_side_R_03, swing1LimitDegrees=8, swing2LimitDegrees=15, twistLimitDegrees=0}
 {
  bone1Name=Base_cloak_back
  bone2Name=pelvis
  swing1LimitDegrees=45
  swing1Motion=Free
  swing2LimitDegrees=45
  swing2Motion=Free
  twistLimitDegrees=45
  twistMotion=Free
 }
 {bone1Name=Dyn_cloak_back_01, bone2Name=Base_cloak_back, swing1LimitDegrees=15, swing2LimitDegrees=35, twistLimitDegrees=0}
 {bone1Name=Dyn_cloak_back_02, bone2Name=Dyn_cloak_back_01, swing1LimitDegrees=5, swing2LimitDegrees=30, twistLimitDegrees=0}
 {bone1Name=Dyn_cloak_back_03, bone2Name=Dyn_cloak_back_02, swing1LimitDegrees=4, swing2LimitDegrees=25, twistLimitDegrees=0}
 {bone1Name=Dyn_cloak_back_04, bone2Name=Dyn_cloak_back_03, swing1LimitDegrees=8, swing2LimitDegrees=25, twistLimitDegrees=0}
 {
  bone1Name=Base_cloak_front_L
  bone2Name=pelvis
  swing1LimitDegrees=45
  swing1Motion=Free
  swing2LimitDegrees=45
  swing2Motion=Free
  twistLimitDegrees=45
  twistMotion=Free
 }
 {bone1Name=Dyn_cloak_front_L_01, bone2Name=Base_cloak_front_L, swing1LimitDegrees=15, swing2LimitDegrees=30, twistLimitDegrees=0}
 {bone1Name=Dyn_cloak_front_L_02, bone2Name=Dyn_cloak_front_L_01, swing1LimitDegrees=10, swing2LimitDegrees=20, twistLimitDegrees=0}
 {bone1Name=Dyn_cloak_front_L_03, bone2Name=Dyn_cloak_front_L_02, swing1LimitDegrees=5, swing2LimitDegrees=15, twistLimitDegrees=0}
 {bone1Name=Dyn_cloak_front_L_04, bone2Name=Dyn_cloak_front_L_03, swing1LimitDegrees=3, swing2LimitDegrees=8, twistLimitDegrees=0}
 {
  bone1Name=calf_l
  bone2Name=thigh_l
  swing1LimitDegrees=45
  swing1Motion=Free
  swing2LimitDegrees=45
  swing2Motion=Free
  twistLimitDegrees=45
  twistMotion=Free
 }
 {
  bone1Name=calf_r
  bone2Name=thigh_r
  swing1LimitDegrees=45
  swing1Motion=Free
  swing2LimitDegrees=45
  swing2Motion=Free
  twistLimitDegrees=45
  twistMotion=Free
 }
 {
  bone1Name=Dyn_cloak_front_R_03
  bone2Name=Dyn_cloak_side_R_03
  swing1LimitDegrees=45
  swing1Motion=Free
  swing2LimitDegrees=45
  swing2Motion=Free
  twistLimitDegrees=45
  twistMotion=Free
 }
 {
  bone1Name=Dyn_cloak_front_R_04
  bone2Name=Dyn_cloak_side_R_04
  swing1LimitDegrees=45
  swing1Motion=Free
  swing2LimitDegrees=45
  swing2Motion=Free
  twistLimitDegrees=45
  twistMotion=Free
 }
]
