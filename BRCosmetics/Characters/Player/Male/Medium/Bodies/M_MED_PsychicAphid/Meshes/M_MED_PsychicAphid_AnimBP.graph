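# Animation Blueprint for the PsychicAphid male cosmetic body, adding secondary motion to the pose copied from its parent mesh.
# A Control Rig processes that pose and caches it for three pelvis-space RigidBody sims (default, DBNO, sliding), selected by state booleans.
# AnimDynamics then swings the amulet chain with spherical chest limits, and fixed bone offsets adjust cloak and cape.
# The EventGraph updates LOD, physics-off, gravity, front-end and sliding variables.
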AnimBlueprint /BRCosmetics/Characters/Player/Male/Medium/Bodies/M_MED_PsychicAphid/Meshes/M_MED_PsychicAphid_AnimBP
target_skeleton=&/M_MED_PsychicAphid_Skeleton
graphs [
 AnimGraph {nodes [
   n1 = RigidBody!(ComponentPose=Local To Component(LocalPose=Use cached pose 'pre sim'()), Alpha, OverrideWorldGravity:Vector="0,0,-980", ExternalForce) -> (Pose) {
    bClampLinearTranslationLimitToRefPose=true
    bOverrideWorldGravity=true
    overridePhysicsAsset=&/M_MED_PsychicAphid_Physics
    simulationSpace=BaseBoneSpace
    alphaBoolBlend {blendOption=Linear}
    baseBoneRef {boneName=pelvis}
    componentLinearAccScale=<0.87 0.45 0.45>
    componentLinearVelScale=<0.87 0.45 0.45>
    overrideWorldGravity {z=-980}
    simSpaceSettings {
     maxAngularAcceleration=500
     maxAngularVelocity=3
     maxLinearAcceleration=400
     maxLinearVelocity=250
     velocityScaleZ=0
     worldAlpha=1
     externalLinearDragV=<0.1 0.1>
    }
   }
   n2 = RigidBody(ComponentPose=Local To Component(LocalPose=Use cached pose 'pre sim'()), OverrideWorldGravity="0,0,-980") {
    bClampLinearTranslationLimitToRefPose=true
    bOverrideWorldGravity=true
    overridePhysicsAsset=&/M_MED_PsychicAphid_Physics_DBNO
    simulationSpace=BaseBoneSpace
    alphaBoolBlend {blendOption=Linear}
    baseBoneRef {boneName=pelvis}
    overrideWorldGravity {z=-980}
    simSpaceSettings {
     maxAngularAcceleration=500
     maxAngularVelocity=3
     maxLinearAcceleration=400
     maxLinearVelocity=250
     velocityScaleZ=0.2
     worldAlpha=0.8
     externalLinearDragV=<0.1 0.1>
    }
   }
   n3 = Blend Poses by bool(
    BlendPose_0=Component To Local(ComponentPose=@n2)
    BlendPose_1=Component To Local(ComponentPose=@n1)
    BlendTime_0=0
    bActiveValue=Get bIsDBNO()
   ) {blendType=Linear}
   n4 = RigidBody(ComponentPose=Local To Component(LocalPose=Use cached pose 'pre sim'()), OverrideWorldGravity="0,0,-980") {
    bClampLinearTranslationLimitToRefPose=true
    bOverrideWorldGravity=true
    overridePhysicsAsset=&/M_MED_PsychicAphid_Physics_sliding
    simulationSpace=BaseBoneSpace
    alphaBoolBlend {blendOption=Linear}
    baseBoneRef {boneName=pelvis}
    overrideWorldGravity {z=-980}
    simSpaceSettings {
     maxAngularAcceleration=500
     maxAngularVelocity=3
     maxLinearAcceleration=400
     maxLinearVelocity=250
     velocityScaleZ=0.2
     worldAlpha=0.8
     externalLinearDragV=<0.1 0.1>
    }
   }
   n5 = Component To Local(ComponentPose=@n4)
   n6 = Get Get is sliding() -> (Get is sliding:bool)
   n7 = Anim Dynamics - Bone: Dyn_amulet_01!(
    ComponentPose=Local To Component(LocalPose=Blend Poses by bool(BlendPose_0=@n5, BlendPose_1=@n3, BlendTime_0=0, bActiveValue=@n6) {blendType=Linear})
    bAlphaBoolEnabled:bool=True
    GravityOverride:Vector="0,0,-980"
    ExternalForce="0,0,0"
   ) -> (Pose) {
    alphaInputType=Bool
    angularDampingOverride=0.9999
    angularSpringConstant=500
    bAngularSpring=true
    bGravityOverrideInSimSpace=true
    bOverrideAngularDamping=true
    bUseGravityOverride=true
    simulationSpace=BoneRelative
    alphaBoolBlend {blendOption=Linear}
    boundBone {boneName=Dyn_amulet_01}
    componentAppliedLinearAccClamp=<3500 3500 3500>
    componentLinearAccScale=<1.2 1.2 0.6>
    componentLinearVelScale=<1.2 1.2 0.6>
    gravityOverride {z=-980}
    physicsBodyDefinitions [{
      sphereCollisionRadius=0
      boundBone {boneName=Dyn_amulet_01}
      boxExtents=<8 8 8>
      constraintSetup {angularLimitsMax=<0 0 15>, angularLimitsMin=<0 -60 -15>, angularTarget=<0 0 0>}
      localJointOffset=<0.6 0 0>
     }]
    relativeSpaceBone {boneName=root}
   }
   n8 = Anim Dynamics - Bone: Dyn_amulet_02!(ComponentPose=@n7, bAlphaBoolEnabled:bool=True, GravityOverride:Vector="0,0,-980", ExternalForce="0,0,0") -> (Pose) {
    alphaInputType=Bool
    angularDampingOverride=0.9999
    angularSpringConstant=500
    bAngularSpring=true
    bGravityOverrideInSimSpace=true
    bOverrideAngularDamping=true
    bUseGravityOverride=true
    simulationSpace=BoneRelative
    alphaBoolBlend {blendOption=Linear}
    boundBone {boneName=Dyn_amulet_02}
    componentAppliedLinearAccClamp=<3500 3500 3500>
    componentLinearAccScale=<1.2 1.2 0.6>
    componentLinearVelScale=<1.2 1.2 0.6>
    gravityOverride {z=-980}
    physicsBodyDefinitions [{
      sphereCollisionRadius=0
      boundBone {boneName=Dyn_amulet_02}
      boxExtents=<8 8 8>
      constraintSetup {angularLimitsMax=<0 -4 5>, angularLimitsMin=<0 -80 -5>, angularTarget=<0 0 0>}
      localJointOffset=<0.6 0 0>
     }]
    relativeSpaceBone {boneName=root}
   }
   n9 = Anim Dynamics - Bone: Dyn_amulet_03!(ComponentPose=@n8, bAlphaBoolEnabled:bool=True, GravityOverride:Vector="0,0,-980", ExternalForce="0,0,0") -> (Pose) {
    alphaInputType=Bool
    angularDampingOverride=0.9999
    angularSpringConstant=500
    bAngularSpring=true
    bGravityOverrideInSimSpace=true
    bOverrideAngularDamping=true
    bUseGravityOverride=true
    bUseSphericalLimits=true
    simulationSpace=BoneRelative
    alphaBoolBlend {blendOption=Linear}
    boundBone {boneName=Dyn_amulet_03}
    componentAppliedLinearAccClamp=<3500 3500 3500>
    componentLinearAccScale=<1.2 1.2 1.2>
    componentLinearVelScale=<1.2 1.2 1.2>
    gravityOverride {z=-980}
    physicsBodyDefinitions [{
      collisionType=CustomSphere
      sphereCollisionRadius=3
      boundBone {boneName=Dyn_amulet_03}
      boxExtents=<8 8 8>
      constraintSetup {angularLimitsMax=<0 0 0>, angularLimitsMin=<0 -20 0>, angularTarget=<0 0 0>}
      localJointOffset=<0.7 0 0>
     }]
    relativeSpaceBone {boneName=root}
    sphericalLimits [{limitRadius=12.3, drivingBone {boneName=spine_04}, sphereLocalOffset=<0 0 0>}]
   }
   n10 = Anim Dynamics - Bone: Dyn_amulet_03(ComponentPose=@n9, bAlphaBoolEnabled=True, GravityOverride="0,0,-980", ExternalForce="0,0,0") {
    alphaInputType=Bool
    angularDampingOverride=0.9999
    angularSpringConstant=500
    bAngularSpring=true
    bGravityOverrideInSimSpace=true
    bOverrideAngularDamping=true
    bUseGravityOverride=true
    bUseSphericalLimits=true
    simulationSpace=BoneRelative
    alphaBoolBlend {blendOption=Linear}
    boundBone {boneName=Dyn_amulet_03}
    componentAppliedLinearAccClamp=<3500 3500 3500>
    componentLinearAccScale=<1.2 1.2 1.2>
    componentLinearVelScale=<1.2 1.2 1.2>
    gravityOverride {z=-980}
    physicsBodyDefinitions [{
      collisionType=CustomSphere
      sphereCollisionRadius=3
      boundBone {boneName=Dyn_amulet_03}
      boxExtents=<8 8 8>
      constraintSetup {angularLimitsMax=<0 -3 0>, angularLimitsMin=<0 -20 0>, angularTarget=<0 0 0>}
      localJointOffset=<0.7 0 0>
     }]
    relativeSpaceBone {boneName=root}
    sphericalLimits [{limitRadius=11, drivingBone {boneName=spine_04}, sphereLocalOffset=<0 0 0>}]
   }
   n11 = "Transform (Modify) Bone - Bone: Base_cloak_front_L"(ComponentPose=@n10, Alpha=1, Translation="0,0,0", Rotation="-3.094,0.571,2.971", Scale="1,1,1") {
    rotationMode=BMM_Additive
    scaleMode=BMM_Additive
    translationMode=BMM_Additive
    alphaBoolBlend {blendOption=Linear}
    boneToModify {boneName=Base_cloak_front_L}
    rotation=<rot -3.094 0.571 2.971>
   }
   n12 = "Transform (Modify) Bone - Bone: Dyn_cloak_front_L_02"(ComponentPose=@n11, Alpha=1, Translation="0,0,0", Rotation="-3.978,-0.920,12.338", Scale="1,1,1") {
    rotationMode=BMM_Additive
    scaleMode=BMM_Additive
    translationMode=BMM_Additive
    alphaBoolBlend {blendOption=Linear}
    boneToModify {boneName=Dyn_cloak_front_L_02}
    rotation=<rot -3.978 -0.92 12.338>
   }
   n13 = "Transform (Modify) Bone - Bone: Dyn_cloak_front_L_03"(ComponentPose=@n12, Alpha=1, Translation="0,0,0", Rotation="5.401,1.329,-6.286", Scale="1,1,1") {
    rotationMode=BMM_Additive
    scaleMode=BMM_Additive
    translationMode=BMM_Additive
    alphaBoolBlend {blendOption=Linear}
    boneToModify {boneName=Dyn_cloak_front_L_03}
    rotation=<rot 5.401 1.329 -6.286>
   }
   n14 = "Transform (Modify) Bone - Bone: Dyn_cloak_front_L_04"(ComponentPose=@n13, Alpha=1, Translation="0,0,0", Rotation="-5.825,-2.681,5.627", Scale="1,1,1") {
    rotationMode=BMM_Additive
    scaleMode=BMM_Additive
    translationMode=BMM_Additive
    alphaBoolBlend {blendOption=Linear}
    boneToModify {boneName=Dyn_cloak_front_L_04}
    rotation=<rot -5.825 -2.681 5.627>
   }
   n15 = "Transform (Modify) Bone - Bone: Base_cloak_front_R"(ComponentPose=@n14, Alpha=1, Translation="0,0,0", Rotation="7,0,-5", Scale="1,1,1") {
    rotationMode=BMM_Additive
    scaleMode=BMM_Additive
    translationMode=BMM_Additive
    alphaBoolBlend {blendOption=Linear}
    boneToModify {boneName=Base_cloak_front_R}
    rotation {pitch=7, roll=-5}
   }
   n16 = "Transform (Modify) Bone - Bone: attach_cape"(ComponentPose=@n15, Alpha=1, Translation="0,0,0", Rotation="0,0,10", Scale="1,1,1") {
    rotationMode=BMM_Additive
    rotationSpace=BCS_BoneSpace
    alphaBoolBlend {blendInTime=0.5, blendOption=Linear, blendOutTime=0.5}
    boneToModify {boneName=attach_cape}
    rotation {roll=10}
   }
   Output Pose(Result=Component To Local(ComponentPose=@n16)) {layerGroup=None}
   n17 = Control Rig!(
    Source=Copy Pose From Mesh() {bUseAttachedParent=true}
    Alpha
    RBAN_off:bool=Get RBAN_off() -> (RBAN_off:bool)
    LOD2:bool=Get LOD2() -> (LOD2:bool)
    LOD1:bool=Get LOD1() -> (LOD1:bool)
    InFrontEnd:bool=false
   ) -> (Pose) {
    alphaBoolBlend {blendOption=Linear}
    controlRigAssetReference {blueprintRigClass=&/M_MED_PsychicAphid_CtrlRig1.M_MED_PsychicAphid_CtrlRig1_C}
   }
   Save cached pose 'pre sim'(Pose=@n17) {cacheName=pre sim}
  ]}
 EventGraph {nodes [
   n1 = Event BlueprintUpdateAnimation()
   > n2 = Gravity Override State Function DEPRECATED(
    CharacterPartAnimInstance=Self-Reference()
    GameTimeInSeconds=GetGameTimeInSeconds()
    GravityOverrideParams=Get Gravity Override Params()
    DebugJointName=dyn_hackBottomArmor
   )
   > Set Override World Gravity(Override World Gravity:Vector=@n2) -> (Output_Get:Vector)
   > Set RBAN_off(RBAN_off:bool=NOT Boolean(A=Get bRigidBodyEnabled())) -> (Output_Get:bool)
   n3 = Get CurrentLOD()
   > Set LOD2(LOD2:bool="integer >= integer"(A=@n3, B=2)) -> (Output_Get:bool)
   > Set LOD1(LOD1:bool="integer >= integer"(A=@n3, B=1)) -> (Output_Get:bool)
   > Set NotFrontEnd(NotFrontEnd:bool=NOT Boolean(A=Get bIsInFrontEnd())) -> (Output_Get:bool)
   n4 = AND Boolean(A=Get bIsInFrontEnd(), B=NOT Boolean(A=Get bIsPlayingEmote()))
   > Set Alpha frontEnd(Alpha frontEnd:double=FInterpTo(
     Current=Get Alpha frontEnd() -> (Alpha frontEnd:double)
     Target="To Float (Boolean)"(InBool=@n4)
     DeltaTime=float * float(A=@n1.DeltaTimeX, B=0.3) {is=title K2Node_PromotableOperator}
     InterpSpeed=10
    )) -> (Output_Get:double)
   Event OnBeginSliding() {is=title K2Node_Event}
   > Set Get is sliding(Get is sliding:bool=true) -> (Output_Get:bool)
   Event OnEndSliding() {is=title K2Node_Event}
   > Set Get is sliding(Get is sliding=false)
  ]}
]
variables [
 Alpha frontEnd {type=number}
 Current LOD {type=integer}
 FrontEndNotEmote {type=boolean}
 Get is sliding {type=boolean}
 Gravity Override Params {
  type=GravityOverrideParamsStruct
  default {
   dBNO {jointName=Root, gravityOverride=<0 150 -980>}
   skydive_FloatingDown {gravityOverride=<0 -700 0>}
   skydive_GoingUp {gravityOverride=<0 -300 -500>}
  }
 }
 LOD1 {type=boolean}
 LOD2 {type=boolean}
 NotFrontEnd {type=boolean}
 Override World Gravity {type=Vector}
 RBAN_off {type=boolean}
]
