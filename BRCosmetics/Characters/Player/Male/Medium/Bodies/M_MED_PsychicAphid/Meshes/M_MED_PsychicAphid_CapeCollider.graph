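# Physics asset for the cape of the medium male PsychicAphid player cosmetic body, providing capsule bodies for both the cloak and the character.
# Five cloak chains (back, front L/R, side L/R) hang from tiny base capsules constrained to the pelvis, with three or four segments each.
# Cloak joints have free swing and twist.
# Spine, pelvis, arm and leg capsules form the body volume, with larger upper and lower arm capsules on the right.
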PhysicsAsset /BRCosmetics/Characters/Player/Male/Medium/Bodies/M_MED_PsychicAphid/Meshes/M_MED_PsychicAphid_CapeCollider
bodies {
 Base_cloak_back [{shapeType=Capsule, length=0.505, radius=0.505, shapeName=Base_cloak_back_capsule, rotation=<rot 0 -90 0>}]
 Base_cloak_front_L [{shapeType=Capsule, length=0.505, radius=0.505, shapeName=Base_cloak_front_L_capsule, rotation=<rot 0 -90 0>}]
 Base_cloak_front_R [{shapeType=Capsule, length=0.505, radius=0.505, shapeName=Base_cloak_front_R_capsule, rotation=<rot 0 -90 0>}]
 Base_cloak_side_L [{shapeType=Capsule, length=0.505, radius=0.505, shapeName=Base_cloak_side_L_capsule, rotation=<rot 0 -90 0>}]
 Base_cloak_side_R [{shapeType=Capsule, length=0.505, radius=0.505, shapeName=Base_cloak_side_R_capsule, rotation=<rot 0 -90 0>}]
 Dyn_cloak_back_01 [{shapeType=Capsule, length=8.7128, radius=10.4059, shapeName=Dyn_cloak_back_01_capsule, center=<7.9445 0 -8.6767>, rotation=<rot 0 0 90>}]
 Dyn_cloak_back_02 [{shapeType=Capsule, length=8.9609, radius=10.1579, shapeName=Dyn_cloak_back_02_capsule, center=<7.9445 0 -5.1826>, rotation=<rot 0 0 90>}]
 Dyn_cloak_back_03 [{
   shapeType=Capsule
   length=8.1887
   radius=11.9301
   shapeName=Dyn_cloak_back_03_capsule
   center=<11.1114 -0.2356 -5.4486>
   rotation=<rot 0 0 90>
  }]
 Dyn_cloak_front_L_01 [{shapeType=Capsule, length=8.7128, radius=10.4059, shapeName=Dyn_cloak_front_L_01_capsule, center=<7.9445 0 -4.74>, rotation=<rot 0 0 90>}]
 Dyn_cloak_front_L_02 [{shapeType=Capsule, length=8.9609, radius=10.1579, shapeName=Dyn_cloak_front_L_02_capsule, center=<7.9445 0 -2.2018>, rotation=<rot 0 0 90>}]
 Dyn_cloak_front_L_03 [{
   shapeType=Capsule
   length=8.1887
   radius=10.9301
   shapeName=Dyn_cloak_front_L_03_capsule
   center=<8.5579 0.7963 -3.8714>
   rotation=<rot 0 0.0001 46.3477>
  }]
 Dyn_cloak_front_R_01 [{shapeType=Capsule, length=8.7128, radius=10.4059, shapeName=Dyn_cloak_front_R_01_capsule, center=<7.9445 0 -8.3511>, rotation=<rot 0 0 90>}]
 Dyn_cloak_front_R_02 [{
   shapeType=Capsule
   length=8.9609
   radius=10.1579
   shapeName=Dyn_cloak_front_R_02_capsule
   center=<7.9445 -3.8694 -2.2018>
   rotation=<rot 0 0 90>
  }]
 Dyn_cloak_front_R_03 [{
   shapeType=Capsule
   length=8.1887
   radius=10.9301
   shapeName=Dyn_cloak_front_R_03_capsule
   center=<8.5579 -2.5423 -3.6399>
   rotation=<rot 0 0 151.3298>
  }]
 Dyn_cloak_front_R_04 [{
   shapeType=Capsule
   length=7.5773
   radius=11.5414
   shapeName=Dyn_cloak_front_R_04_capsule
   center=<8.1398 -5.8772 -4.6903>
   rotation=<rot 0 0 135.8018>
  }]
 Dyn_cloak_side_L_01 [{shapeType=Capsule, length=8.7128, radius=10.4059, shapeName=Dyn_cloak_side_L_01_capsule, center=<7.9445 0 -7.2997>, rotation=<rot 0 0 90>}]
 Dyn_cloak_side_L_02 [{shapeType=Capsule, length=8.9609, radius=10.1579, shapeName=Dyn_cloak_side_L_02_capsule, center=<8.5579 0 -7.6624>, rotation=<rot 0 0 90>}]
 Dyn_cloak_side_L_03 [{shapeType=Capsule, length=7.5773, radius=11.5414, shapeName=Dyn_cloak_side_L_03_capsule, center=<8.1398 0 -8.1939>, rotation=<rot 0 0 90>}]
 Dyn_cloak_side_R_01 [{shapeType=Capsule, length=8.7128, radius=10.4059, shapeName=Dyn_cloak_side_R_01_capsule, center=<7.9445 0 -9.4178>, rotation=<rot 0 0 90>}]
 Dyn_cloak_side_R_02 [{shapeType=Capsule, length=8.9609, radius=10.1579, shapeName=Dyn_cloak_side_R_02_capsule, center=<7.9445 0 -7.4399>, rotation=<rot 0 0 90>}]
 Dyn_cloak_side_R_03 [{shapeType=Capsule, length=8.1887, radius=10.9301, shapeName=Dyn_cloak_side_R_03_capsule, center=<8.5579 0 -7.5721>, rotation=<rot 0 0 90>}]
 Dyn_cloak_side_R_04 [{
   shapeType=Capsule
   length=7.5773
   radius=14.0414
   shapeName=Dyn_cloak_side_R_04_capsule
   center=<10.2348 0.2854 -9.7705>
   rotation=<rot 0 -7.3429 90>
  }]
 calf_l [{shapeType=Capsule, length=27.8078, radius=11.1974, center=<-16.4529 -2.2427 2.1173>, rotation=<rot 90 180 180>}]
 calf_r [{shapeType=Capsule, length=27.8078, radius=11.1974, center=<16.4529 2.2427 -2.1173>, rotation=<rot 90 180 0>}]
 foot_l [{shapeType=Capsule, length=11.3965, radius=12.7763, center=<-4.5038 7.2322 1.578>, rotation=<rot 0 14.9998 90>}]
 foot_r [{shapeType=Capsule, length=11.3965, radius=12.7763, center=<4.5038 -7.2322 -1.578>, rotation=<rot 0 14.9998 -90>}]
 hand_l [{shapeType=Capsule, length=9.5015, radius=6.7933, center=<6.2703 0 0>, rotation=<rot 90 0 0>}]
 hand_r [{shapeType=Capsule, length=9.5015, radius=6.7933, center=<-6.2703 0 0>, rotation=<rot 90 0 180>}]
 lowerarm_l [{shapeType=Capsule, length=14.3387, radius=7.5574, center=<8.9852 0 0.2267>, rotation=<rot 90 0 0>}]
 lowerarm_r [{shapeType=Capsule, length=19.3387, radius=9.5574, center=<-8.9852 0 -0.2267>, rotation=<rot 90 0 180>}]
 pelvis [{shapeType=Capsule, length=6.7241, radius=16.0874, center=<-2.4855 -0.9262 0>}]
 spine_02 [{shapeType=Capsule, length=3.3213, radius=14.7784, center=<-2.6142 1.0379 0>}]
 spine_03 [{shapeType=Capsule, length=4.9972, radius=14.7226, center=<0 1.2208 0>}]
 spine_04 [{shapeType=Capsule, length=3.6167, radius=16.287, center=<0 1.1241 0>}]
 spine_05 [{shapeType=Capsule, length=14.0449, radius=15.8769, center=<0 1.0463 0>}]
 thigh_l [{shapeType=Capsule, length=18.7606, radius=11.2945, center=<-18.4217 -1.038 0>, rotation=<rot 90 180 2.664>}]
 thigh_r [{shapeType=Capsule, length=18.7606, radius=11.2945, center=<18.4217 1.038 0>, rotation=<rot 90 179.7762 -177.5598>}]
 upperarm_l [{shapeType=Capsule, length=17.1712, radius=7.7501, center=<10.0455 0 0>, rotation=<rot 90 0 0>}]
 upperarm_r [{shapeType=Capsule, length=21.1712, radius=9.7501, center=<-10.0455 0 0>, rotation=<rot 90 0 180>}]
}
constraints [
 {bone1Name=spine_02, bone2Name=pelvis, swing1LimitDegrees=45, swing2LimitDegrees=45, twistLimitDegrees=45}
 {bone1Name=spine_03, bone2Name=spine_02, swing1LimitDegrees=45, swing2LimitDegrees=45, twistLimitDegrees=45}
 {bone1Name=spine_04, bone2Name=spine_03, swing1LimitDegrees=45, swing2LimitDegrees=45, twistLimitDegrees=45}
 {bone1Name=spine_05, bone2Name=spine_04, swing1LimitDegrees=45, swing2LimitDegrees=45, twistLimitDegrees=45}
 {bone1Name=upperarm_l, bone2Name=spine_05, swing1LimitDegrees=45, swing2LimitDegrees=45, twistLimitDegrees=45}
 {bone1Name=lowerarm_l, bone2Name=upperarm_l, swing1LimitDegrees=45, swing2LimitDegrees=45, twistLimitDegrees=45}
 {bone1Name=hand_l, bone2Name=lowerarm_l, swing1LimitDegrees=45, swing2LimitDegrees=45, twistLimitDegrees=45}
 {bone1Name=thigh_l, bone2Name=pelvis, swing1LimitDegrees=45, swing2LimitDegrees=45, twistLimitDegrees=45}
 {bone1Name=hand_r, bone2Name=lowerarm_r, swing1LimitDegrees=45, swing2LimitDegrees=45, twistLimitDegrees=45}
 {bone1Name=thigh_r, bone2Name=pelvis, swing1LimitDegrees=45, swing2LimitDegrees=45, twistLimitDegrees=45}
 {bone1Name=upperarm_r, bone2Name=spine_05, swing1LimitDegrees=45, swing2LimitDegrees=45, twistLimitDegrees=45}
 {bone1Name=lowerarm_r, bone2Name=upperarm_r, swing1LimitDegrees=45, swing2LimitDegrees=45, twistLimitDegrees=45}
 {
  bone1Name=Dyn_cloak_front_L_01
  bone2Name=Base_cloak_front_L
  swing1LimitDegrees=45
  swing1Motion=Free
  swing2LimitDegrees=45
  swing2Motion=Free
  twistLimitDegrees=45
  twistMotion=Free
 }
 {
  bone1Name=Dyn_cloak_front_L_02
  bone2Name=Dyn_cloak_front_L_01
  swing1LimitDegrees=45
  swing1Motion=Free
  swing2LimitDegrees=45
  swing2Motion=Free
  twistLimitDegrees=45
  twistMotion=Free
 }
 {
  bone1Name=Dyn_cloak_front_L_03
  bone2Name=Dyn_cloak_front_L_02
  swing1LimitDegrees=45
  swing1Motion=Free
  swing2LimitDegrees=45
  swing2Motion=Free
  twistLimitDegrees=45
  twistMotion=Free
 }
 {
  bone1Name=Dyn_cloak_side_L_03
  bone2Name=Dyn_cloak_side_L_02
  swing1LimitDegrees=45
  swing1Motion=Free
  swing2LimitDegrees=45
  swing2Motion=Free
  twistLimitDegrees=45
  twistMotion=Free
 }
 {
  bone1Name=Dyn_cloak_side_L_01
  bone2Name=Base_cloak_side_L
  swing1LimitDegrees=45
  swing1Motion=Free
  swing2LimitDegrees=45
  swing2Motion=Free
  twistLimitDegrees=45
  twistMotion=Free
 }
 {
  bone1Name=Dyn_cloak_side_L_02
  bone2Name=Dyn_cloak_side_L_01
  swing1LimitDegrees=45
  swing1Motion=Free
  swing2LimitDegrees=45
  swing2Motion=Free
  twistLimitDegrees=45
  twistMotion=Free
 }
 {
  bone1Name=Dyn_cloak_back_02
  bone2Name=Dyn_cloak_back_01
  swing1LimitDegrees=45
  swing1Motion=Free
  swing2LimitDegrees=45
  swing2Motion=Free
  twistLimitDegrees=45
  twistMotion=Free
 }
 {
  bone1Name=Dyn_cloak_back_03
  bone2Name=Dyn_cloak_back_02
  swing1LimitDegrees=45
  swing1Motion=Free
  swing2LimitDegrees=45
  swing2Motion=Free
  twistLimitDegrees=45
  twistMotion=Free
 }
 {
  bone1Name=Dyn_cloak_back_01
  bone2Name=Base_cloak_back
  swing1LimitDegrees=45
  swing1Motion=Free
  swing2LimitDegrees=45
  swing2Motion=Free
  twistLimitDegrees=45
  twistMotion=Free
 }
 {
  bone1Name=Dyn_cloak_side_R_01
  bone2Name=Base_cloak_side_R
  swing1LimitDegrees=45
  swing1Motion=Free
  swing2LimitDegrees=45
  swing2Motion=Free
  twistLimitDegrees=45
  twistMotion=Free
 }
 {
  bone1Name=Dyn_cloak_side_R_02
  bone2Name=Dyn_cloak_side_R_01
  swing1LimitDegrees=45
  swing1Motion=Free
  swing2LimitDegrees=45
  swing2Motion=Free
  twistLimitDegrees=45
  twistMotion=Free
 }
 {
  bone1Name=Dyn_cloak_side_R_03
  bone2Name=Dyn_cloak_side_R_02
  swing1LimitDegrees=45
  swing1Motion=Free
  swing2LimitDegrees=45
  swing2Motion=Free
  twistLimitDegrees=45
  twistMotion=Free
 }
 {
  bone1Name=Dyn_cloak_side_R_04
  bone2Name=Dyn_cloak_side_R_03
  swing1LimitDegrees=45
  swing1Motion=Free
  swing2LimitDegrees=45
  swing2Motion=Free
  twistLimitDegrees=45
  twistMotion=Free
 }
 {
  bone1Name=Dyn_cloak_front_R_01
  bone2Name=Base_cloak_front_R
  swing1LimitDegrees=45
  swing1Motion=Free
  swing2LimitDegrees=45
  swing2Motion=Free
  twistLimitDegrees=45
  twistMotion=Free
 }
 {
  bone1Name=Dyn_cloak_front_R_02
  bone2Name=Dyn_cloak_front_R_01
  swing1LimitDegrees=45
  swing1Motion=Free
  swing2LimitDegrees=45
  swing2Motion=Free
  twistLimitDegrees=45
  twistMotion=Free
 }
 {
  bone1Name=Dyn_cloak_front_R_03
  bone2Name=Dyn_cloak_front_R_02
  swing1LimitDegrees=45
  swing1Motion=Free
  swing2LimitDegrees=45
  swing2Motion=Free
  twistLimitDegrees=45
  twistMotion=Free
 }
 {
  bone1Name=Dyn_cloak_front_R_04
  bone2Name=Dyn_cloak_front_R_03
  swing1LimitDegrees=45
  swing1Motion=Free
  swing2LimitDegrees=45
  swing2Motion=Free
  twistLimitDegrees=45
  twistMotion=Free
 }
 {
  bone1Name=calf_r
  bone2Name=thigh_r
  swing1LimitDegrees=45
  swing1Motion=Free
  swing2LimitDegrees=45
  swing2Motion=Free
  twistLimitDegrees=45
  twistMotion=Free
 }
 {
  bone1Name=Base_cloak_side_L
  bone2Name=pelvis
  swing1LimitDegrees=45
  swing1Motion=Free
  swing2LimitDegrees=45
  swing2Motion=Free
  twistLimitDegrees=45
  twistMotion=Free
 }
 {
  bone1Name=Base_cloak_front_R
  bone2Name=pelvis
  swing1LimitDegrees=45
  swing1Motion=Free
  swing2LimitDegrees=45
  swing2Motion=Free
  twistLimitDegrees=45
  twistMotion=Free
 }
 {
  bone1Name=Base_cloak_front_L
  bone2Name=pelvis
  swing1LimitDegrees=45
  swing1Motion=Free
  swing2LimitDegrees=45
  swing2Motion=Free
  twistLimitDegrees=45
  twistMotion=Free
 }
 {
  bone1Name=Base_cloak_back
  bone2Name=pelvis
  swing1LimitDegrees=45
  swing1Motion=Free
  swing2LimitDegrees=45
  swing2Motion=Free
  twistLimitDegrees=45
  twistMotion=Free
 }
 {
  bone1Name=Base_cloak_side_R
  bone2Name=pelvis
  swing1LimitDegrees=45
  swing1Motion=Free
  swing2LimitDegrees=45
  swing2Motion=Free
  twistLimitDegrees=45
  twistMotion=Free
 }
 {bone1Name=foot_r, bone2Name=calf_r, swing1LimitDegrees=45, swing2LimitDegrees=45, twistLimitDegrees=45}
 {
  bone1Name=calf_l
  bone2Name=thigh_l
  swing1LimitDegrees=45
  swing1Motion=Free
  swing2LimitDegrees=45
  swing2Motion=Free
  twistLimitDegrees=45
  twistMotion=Free
 }
 {bone1Name=foot_l, bone2Name=calf_l, swing1LimitDegrees=45, swing2LimitDegrees=45, twistLimitDegrees=45}
]
