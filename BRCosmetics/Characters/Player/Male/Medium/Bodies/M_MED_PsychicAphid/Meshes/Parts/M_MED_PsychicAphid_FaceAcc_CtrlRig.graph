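# Control Rig for the Psychic Aphid male medium FaceAcc part that applies corrective offsets to hair bones, mainly at the back.
# A sequence layers two head-versus-spine_05 angle ranges, then a correction weighted by the product of both upper-arm spherical pose readers.
# When "RBAN off" is true, a final branch also offsets Dyn hair bones and Master_hair_back.
# Pelvis-height weighting and some modify nodes are left outside the execution chain.
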ControlRigBlueprint /BRCosmetics/Characters/Player/Male/Medium/Bodies/M_MED_PsychicAphid/Meshes/Parts/M_MED_PsychicAphid_FaceAcc_CtrlRig
graphs [
 Multiply_1_1_ContainedGraph {nodes [
   n1 = RigVMFunctionEntryNode()
   RigVMFunctionReturnNode(Result=&RigVMFunction_MathDoubleMul(B=@n1.C, A=&RigVMFunction_MathDoubleMul(A=@n1.A, B=@n1.B).Result).Result)
  ]}
 RigVMFunctionLibrary {links [], nodes []}
 RigVMFunction_Sequence_1_ContainedGraph {nodes [
   n1 = RigVMFunctionEntryNode()
   RigVMFunction_Sequence = &RigVMFunction_Sequence(ExecuteContext=@n1.ExecuteContext)
   n2 = &RigVMFunction_Sequence(ExecuteContext=@RigVMFunction_Sequence.B)
   n3 = &RigVMFunction_Sequence(ExecuteContext=@n2.B)
   RigVMFunctionReturnNode(A=@RigVMFunction_Sequence.A, B=@n2.A, C=@n3.A, D=@n3.B)
  ]}
 RigVMModel {nodes [
   RigUnit_BeginExecution = &RigUnit_BeginExecution()
   RigVMFunction_Sequence = RigVMAggregateNode(ExecuteContext=@RigUnit_BeginExecution.ExecutePin)
   n1 = &RigUnit_SphericalPoseReader(
    ActiveRegionScaleFactors=(PositiveWidth=1,NegativeWidth=1,PositiveHeight=1,NegativeHeight=1)
    ActiveRegionSize=0.1
    Debug=(bDrawDebug=True,bDraw2D=False,bDrawLocalAxes=False,DebugScale=25,DebugSegments=20,DebugThickness=0.25)
    DriverAxis=(X=1,Y=0,Z=0)
    DriverItem=(Type=Bone,Name="upperarm_l")
    FalloffRegionScaleFactors=(PositiveWidth=1,NegativeWidth=1,PositiveHeight=1,NegativeHeight=1)
    FalloffSize=0.2
    FlipHeightScaling=False
    FlipWidthScaling=False
    OptionalParentItem=(Type=Bone,Name="None")
    RotationOffset=(X=0,Y=-70,Z=0)
    ExecutePin=@RigVMFunction_Sequence.C
   )
   > n2 = &RigUnit_SphericalPoseReader(
    ActiveRegionScaleFactors=(PositiveWidth=1,NegativeWidth=1,PositiveHeight=1,NegativeHeight=1)
    ActiveRegionSize=0.1
    Debug=(bDrawDebug=True,bDraw2D=False,bDrawLocalAxes=False,DebugScale=25,DebugSegments=20,DebugThickness=0.25)
    DriverAxis=(X=-1,Y=0,Z=0)
    DriverItem=(Type=Bone,Name="upperarm_r")
    FalloffRegionScaleFactors=(PositiveWidth=1,NegativeWidth=1,PositiveHeight=1,NegativeHeight=1)
    FalloffSize=0.2
    FlipHeightScaling=False
    FlipWidthScaling=False
    OptionalParentItem=(Type=Bone,Name="None")
    RotationOffset=(X=0,Y=105,Z=0)
   )
   n3 = &RigVMFunction_MathDoubleMul(B=@n1.OutputParam, A=@n2.OutputParam)
   > &RigUnit_ModifyTransforms(
    ItemToModify=((Item=(Type=Bone,Name="Base_hair_back_05"),Transform=(Rotation=(X=-0.0189,Y=-0.0851,Z=-0.2156,W=0.9726),Translation=(X=0,Y=0,Z=0),Scale3D=(X=1,Y=1,Z=1))))
    WeightMinimum=0
    Weight=@n3.Result
   )
   > &RigUnit_ModifyTransforms(
    ItemToModify=((Item=(Type=Bone,Name="Base_hair_back_01"),Transform=(Rotation=(X=0,Y=0,Z=0.1736,W=0.9848),Translation=(X=0,Y=0,Z=0),Scale3D=(X=1,Y=1,Z=1))))
    WeightMinimum=0
    Weight=@n3.Result
   )
   > &RigUnit_ModifyTransforms(
    ItemToModify=((Item=(Type=Bone,Name="Base_hair_back_02"),Transform=(Rotation=(X=0,Y=0,Z=0.0872,W=0.9962),Translation=(X=0,Y=0,Z=0),Scale3D=(X=1,Y=1,Z=1))))
    WeightMinimum=0
    Weight=@n3.Result
   )
   n4 = &RigVMFunction_MathDoubleRemap(
    SourceMaximum=-90
    SourceMinimum=-50
    TargetMaximum=1
    TargetMinimum=0
    Value=&RigVMFunction_MathQuaternionToEuler(
     RotationOrder=YXZ
     Value=&RigUnit_GetRelativeTransformForItem(Child=(Type=Bone,Name="head"), Parent=(Type=Bone,Name="spine_05"), bChildInitial=False, bParentInitial=False).RelativeTransform.Rotation
    ).Result.Z
   )
   &RigUnit_ModifyTransforms(
    ItemToModify=((Item=(Type=Bone,Name="Master_hair_back"),Transform=(Rotation=(X=-0.0858,Y=-0.173,Z=-0.0151,W=0.9811),Translation=(X=-1.5,Y=0,Z=0),Scale3D=(X=1,Y=1,Z=1))))
    WeightMinimum=0
    Weight=@n4.Result
    ExecutePin=@RigVMFunction_Sequence.B
   )
   > &RigUnit_ModifyTransforms(
    ItemToModify=((Item=(Type=Bone,Name="Base_hair_back_02"),Transform=(Rotation=(X=0,Y=-0.0872,Z=0,W=0.9962),Translation=(X=0,Y=0,Z=0),Scale3D=(X=1,Y=1,Z=1))))
    WeightMinimum=0
    Weight=@n4.Result
   )
   > &RigUnit_ModifyTransforms(
    ItemToModify=((Item=(Type=Bone,Name="Base_hair_back_01"),Transform=(Rotation=(X=-0.0038,Y=0.0435,Z=0.0871,W=0.9952),Translation=(X=0,Y=0,Z=0),Scale3D=(X=1,Y=1,Z=1))))
    WeightMinimum=0
    Weight=@n4.Result
   )
   n5 = &RigVMFunction_MathDoubleRemap(
    SourceMaximum=-45
    SourceMinimum=-15
    TargetMaximum=1
    TargetMinimum=0
    Value=&RigVMFunction_MathQuaternionToEuler(
     RotationOrder=YXZ
     Value=&RigUnit_GetRelativeTransformForItem(Child=(Type=Bone,Name="head"), Parent=(Type=Bone,Name="spine_05"), bChildInitial=False, bParentInitial=False).RelativeTransform.Rotation
    ).Result.Z
   )
   &RigUnit_ModifyTransforms(
    ItemToModify=((Item=(Type=Bone,Name="Master_hair_back"),Transform=(Rotation=(X=0,Y=-0.0436,Z=0,W=0.999),Translation=(X=0,Y=0,Z=0),Scale3D=(X=1,Y=1,Z=1))))
    WeightMinimum=0
    Weight=@n5.Result
    ExecutePin=@RigVMFunction_Sequence.A
   )
   > &RigUnit_ModifyTransforms(
    ItemToModify=((Item=(Type=Bone,Name="Base_hair_back_01"),Transform=(Rotation=(X=0,Y=-0.1305,Z=0,W=0.9914),Translation=(X=0,Y=0,Z=0),Scale3D=(X=1,Y=1,Z=1))))
    WeightMinimum=0
    Weight=@n5.Result
   )
   RigVMFunction_ControlFlowBranch = &RigVMFunction_ControlFlowBranch(Condition=RigVMVariableNode() {variable=RBAN off}.Value, ExecuteContext=@RigVMFunction_Sequence.D)
   &RigUnit_ModifyTransforms(
    ItemToModify=((Item=(Type=Bone,Name="Dyn_hair_back_05"),Transform=(Rotation=(X=0,Y=-0.0698,Z=0,W=0.9976),Translation=(X=0,Y=0,Z=0),Scale3D=(X=1,Y=1,Z=1))))
    Weight=1
    WeightMinimum=0
    ExecutePin=@RigVMFunction_ControlFlowBranch.True
   )
   n6 = &RigVMFunction_MathQuaternionToEuler(
    RotationOrder=YXZ
    Value=&RigUnit_GetRelativeTransformForItem(Child=(Type=Bone,Name="head"), Parent=(Type=Bone,Name="spine_05"), bChildInitial=False, bParentInitial=False).RelativeTransform.Rotation
   )
   n7 = &RigVMFunction_MathDoubleRemap(SourceMaximum=-58, SourceMinimum=-20, TargetMaximum=1, TargetMinimum=0, Value=@n6.Result.Z)
   > &RigUnit_ModifyTransforms(
    ItemToModify=((Item=(Type=Bone,Name="Master_hair_back"),Transform=(Rotation=(X=-0.1261,Y=-0.2566,Z=-0.0338,W=0.9577),Translation=(X=-1.5,Y=0,Z=0),Scale3D=(X=0.9,Y=0.9,Z=0.9))))
    WeightMinimum=0
    Weight=@n7.Result
   )
   > &RigUnit_ModifyTransforms(
    ItemToModify=((Item=(Type=Bone,Name="Dyn_hair_up_bk_06"),Transform=(Rotation=(X=0,Y=-0.2588,Z=0,W=0.9659),Translation=(X=0,Y=0,Z=0),Scale3D=(X=1,Y=1,Z=1))))
    WeightMinimum=0
    Weight=@n7.Result
   )
   n8 = &RigVMFunction_MathDoubleRemap(SourceMaximum=-90, SourceMinimum=-50, TargetMaximum=1, TargetMinimum=0, Value=@n6.Result.Z)
   > &RigUnit_ModifyTransforms(
    ItemToModify=((Item=(Type=Bone,Name="Master_hair_back"),Transform=(Rotation=(X=0,Y=-0.1305,Z=0,W=0.9914),Translation=(X=0,Y=0,Z=0),Scale3D=(X=1,Y=1,Z=1))))
    WeightMinimum=0
    Weight=@n8.Result
   )
   > &RigUnit_ModifyTransforms(
    ItemToModify=((Item=(Type=Bone,Name="Dyn_hair_up_bk_06"),Transform=(Rotation=(X=0.0226,Y=-0.2578,Z=0.0842,W=0.9623),Translation=(X=0,Y=0,Z=0),Scale3D=(X=1,Y=1,Z=1))))
    WeightMinimum=0
    Weight=@n8.Result
   )
   n9 = &RigUnit_GetRelativeTransformForItem(Child=(Type=Bone,Name="head"), Parent=(Type=Bone,Name="spine_05"), bChildInitial=False, bParentInitial=False)
   Multiply = &RigVMFunction_MathDoubleMul(
    A=&RigVMFunction_MathDoubleRemap(
     SourceMaximum=-55
     SourceMinimum=-25
     TargetMaximum=1
     TargetMinimum=0
     Value=&RigVMFunction_MathQuaternionToEuler(RotationOrder=YXZ, Value=@n9.RelativeTransform.Rotation).Result.Z
    ).Result
    B=&RigVMFunction_MathDoubleRemap(
     SourceMaximum=45
     SourceMinimum=50
     TargetMaximum=1
     TargetMinimum=0
     Value=&RigUnit_GetRelativeTransformForItem(Child=(Type=Bone,Name="pelvis"), Parent=(Type=Bone,Name="root"), bChildInitial=False, bParentInitial=False).RelativeTransform.Translation.Z
    ).Result
   )
   &RigUnit_ModifyTransforms(
    ExecutePin=()
    ItemToModify=((Item=(Type=Bone,Name="Master_hair_back"),Transform=(Rotation=(X=0,Y=0,Z=0,W=1),Translation=(X=0,Y=0,Z=0),Scale3D=(X=1,Y=1,Z=1))))
    WeightMinimum=0
    Weight=@Multiply.Result
   )
   &RigUnit_ModifyTransforms(
    ExecutePin=()
    ItemToModify=((Item=(Type=Bone,Name="Base_hair_up_bk_06"),Transform=(Rotation=(X=0,Y=-0.766,Z=0,W=0.6428),Translation=(X=0,Y=0,Z=0),Scale3D=(X=1,Y=1,Z=1))))
    WeightMinimum=0
    Weight=@n4.Result
   )
   Multiply_1 = RigVMAggregateNode(
    B=&RigVMFunction_MathDoubleRemap(
     SourceMaximum=70
     SourceMinimum=75
     TargetMaximum=1
     TargetMinimum=0
     Value=&RigUnit_GetRelativeTransformForItem(Child=(Type=Bone,Name="pelvis"), Parent=(Type=Bone,Name="root"), bChildInitial=False, bParentInitial=False).RelativeTransform.Translation.Z
    ).Result
    C=&RigVMFunction_MathDoubleRemap(
     SourceMaximum=65
     SourceMinimum=60
     TargetMaximum=1
     TargetMinimum=0
     Value=&RigUnit_GetRelativeTransformForItem(Child=(Type=Bone,Name="pelvis"), Parent=(Type=Bone,Name="root"), bChildInitial=False, bParentInitial=False).RelativeTransform.Translation.Z
    ).Result
    A=@n4.Result
   )
  ]}
]
hierarchy [
 root {loc=<0 0 0>, rot=<0 0 0 1>}
 pelvis {parent=root, loc=<0 -0.7071 93.1191>, rot=<0.0127 0.707 -0.0127 -0.707>}
 spine_01 {parent=pelvis, loc=<6.7494 -0.6388 0>, rot=<0 0 0.034 -0.9994>}
 spine_02 {parent=spine_01, loc=<7.2518 -0.035 0>, rot=<0 0 0.0238 -0.9997>}
 spine_03 {parent=spine_02, loc=<6.0104 0 0>, rot=<0 0 0.0149 -0.9999>}
 spine_04 {parent=spine_03, loc=<8.1024 0 0>, rot=<0 0 -0.0132 -0.9999>}
 spine_05 {parent=spine_04, loc=<10.0146 0 0>, rot=<0 0 -0.0015 -1>}
 neck_01 {parent=spine_05, loc=<11.3138 0.1529 0>, rot=<0 0 -0.2003 -0.9797>}
 neck_02 {parent=neck_01, loc=<4.5627 0 0>, rot=<0 0 0.0504 -0.9987>}
 head {parent=neck_02, loc=<5.3914 0 0>, rot=<0 0 0.115 -0.9934>}
 Dyn_earring_L {parent=head, loc=<0.7041 0.9526 -6.1316>, rot=<0.0041 -1 0 0>}
 Dyn_hair_front_01 {parent=head, loc=<10.8173 9.0278 2.3788>, rot=<0.3691 0.3312 0.8673 0.0427>}
 Dyn_hair_front_02 {parent=head, loc=<12.8273 5.6481 4.8128>, rot=<0.3208 0.2643 0.9094 0.0144>}
 Dyn_hair_side_01 {parent=head, loc=<9.2413 4.4659 6.2579>, rot=<0.1357 0.258 0.956 0.0329>}
 Dyn_hair_side_02 {parent=head, loc=<10.3851 1.5731 7.2563>, rot=<0.2234 0.1224 0.967 0.008>}
 Dyn_hair_side_03 {parent=head, loc=<7.9137 -0.4278 7.9203>, rot=<0.0879 -0.0193 0.9941 0.06>}
 Dyn_hair_up_01 {parent=head, loc=<13.2519 7.5904 0.2575>, rot=<0.7421 0.2373 0.6154 0.1192>}
 Dyn_hair_up_02 {parent=head, loc=<13.9778 1.4841 -1.0842>, rot=<0.851 0.1441 0.4571 0.2146>}
 Dyn_hair_up_03 {parent=head, loc=<12.4157 -3.1473 -1.1168>, rot=<0.8656 0.0371 0.1109 -0.4869>}
 Master_hair_back {parent=head, loc=<8.1048 -5.0998 -0.452>, rot=<0.1426 -0.7102 0.6651 -0.1815>}
 Base_hair_back_01 {parent=Master_hair_back, loc=<0.6469 -4.2979 -0.3365>, rot=<-0.3672 -0.1554 0.2145 -0.8916>}
 Dyn_hair_back_01 {parent=Base_hair_back_01, loc=<0 0 0>, rot=<0 0 0 1>}
 Base_hair_back_02 {parent=Master_hair_back, loc=<1.7774 -1.9001 1.4855>, rot=<-0.1716 -0.1681 0.1421 -0.9603>}
 Dyn_hair_back_02 {parent=Base_hair_back_02, loc=<0 0 0>, rot=<0 0 0 1>}
 Base_hair_back_04 {parent=Master_hair_back, loc=<2.0586 0.5805 0.8735>, rot=<0.1046 -0.2101 0.0061 -0.972>}
 Dyn_hair_back_04 {parent=Base_hair_back_04, loc=<0 0 0>, rot=<0 0 0 1>}
 Base_hair_back_05 {parent=Master_hair_back, loc=<2.138 3.1155 0.93>, rot=<0.1854 -0.0857 -0.1521 -0.967>}
 Dyn_hair_back_05 {parent=Base_hair_back_05, loc=<0 0 0>, rot=<0 0 0 1>}
 base_hair_L_01 {parent=Master_hair_back, loc=<-0.4951 4.718 -0.5966>, rot=<0.4344 -0.1297 -0.2012 -0.8683>}
 Dyn_hair_L_01 {parent=base_hair_L_01, loc=<0 0 0>, rot=<0 0 0 1>}
 Dyn_hair_up_04 {parent=head, loc=<12.4114 -1.3576 5.2847>, rot=<0.5295 -0.0396 0.8474 0.0016>}
 Dyn_hair_up_bk_05 {parent=head, loc=<11.9554 -3.4646 3.1065>, rot=<0.5843 -0.1707 0.7834 -0.1256>}
 Dyn_hair_up_06 {parent=head, loc=<14.0318 1.7051 3.9113>, rot=<0.5115 0.069 0.8157 0.2611>}
 Dyn_hair_up_07 {parent=head, loc=<13.6962 4.8969 2.2795>, rot=<0.8086 0.0562 0.5783 0.0926>}
 Dyn_hair_up_L_08 {parent=head, loc=<11.9488 6.8115 -2.3287>, rot=<0.9455 0.2399 0.0866 -0.2027>}
 Dyn_hair_up_L_09 {parent=head, loc=<13.2639 1.4065 -3.9814>, rot=<0.9939 0.0244 0.0549 -0.0926>}
 Base_hair_side_04 {parent=head, loc=<8.1134 -3.3138 5.9329>, rot=<0.1274 -0.2006 0.9711 0.024>}
 Dyn_hair_side_04 {parent=Base_hair_side_04, loc=<0 0 0>, rot=<0 0 0 1>}
 Base_hair_up_bk_06 {parent=head, loc=<11.0742 -5.2696 2.1924>, rot=<0.3978 -0.4646 0.7622 -0.2122>}
 Dyn_hair_up_bk_06 {parent=Base_hair_up_bk_06, loc=<0 0 0>, rot=<0 0 0 1>}
 clavicle_l {parent=spine_05, loc=<4.3575 -0.1466 -2.3223>, rot=<0.0465 -0.7289 -0.009 -0.6829>}
 upperarm_l {parent=clavicle_l, loc=<15.5111 0 0>, rot=<0.0226 -0.3234 -0.0088 -0.9459>}
 lowerarm_l {parent=upperarm_l, loc=<26.6514 0 0>, rot=<0.0391 -0.0099 -0.2453 -0.9686>}
 upperarm_twist_01_l {parent=upperarm_l, loc=<0.0041 0 0>, rot=<0 0 0 1>}
 upperarm_twist_02_l {parent=upperarm_l, loc=<13.7941 0 0>, rot=<0 0 0 1>}
 clavicle_r {parent=spine_05, loc=<4.3575 -0.1466 2.3223>, rot=<0.7289 0.0465 -0.6829 0.009>}
 upperarm_r {parent=clavicle_r, loc=<-15.5111 0 0>, rot=<0.0226 -0.3234 -0.0088 -0.9459>}
 lowerarm_r {parent=upperarm_r, loc=<-26.6514 0 0>, rot=<0.0391 -0.0099 -0.2453 -0.9686>}
 upperarm_twist_01_r {parent=upperarm_r, loc=<-0.0041 0 0>, rot=<0 0 0 -1>}
 upperarm_twist_02_r {parent=upperarm_r, loc=<-13.7941 0 0>, rot=<0 0 0 -1>}
]
variables [RBAN off {type=bool, default=False}]
