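# Physics asset for the face accessory mesh of the male medium PsychicAphid player cosmetic, setting up capsule bodies and joints for its hair.
# Head, neck, upper spine and clavicle capsules form the body volume, linked by free-swinging constraints.
# Dynamic hair capsules hang off the head directly or through tiny base bodies, several of which chain through a master back-hair body.
# Their joints allow small swing angles with zero twist limit, while base and master joints are free.
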PhysicsAsset /BRCosmetics/Characters/Player/Male/Medium/Bodies/M_MED_PsychicAphid/Meshes/Parts/M_MED_PsychicAphid_FaceAcc_Physics
bodies {
 Base_hair_back_01 [{shapeType=Capsule, length=0.505, radius=0.505, shapeName=Base_hair_back_01_capsule, rotation=<rot 0 -90 0>}]
 Base_hair_back_02 [{shapeType=Capsule, length=0.505, radius=0.505, shapeName=Base_hair_back_02_capsule, rotation=<rot 0 -90 0>}]
 Base_hair_back_04 [{shapeType=Capsule, length=0.505, radius=0.505, shapeName=Base_hair_back_04_capsule, rotation=<rot 0 -90 0>}]
 Base_hair_back_05 [{shapeType=Capsule, length=0.505, radius=0.505, shapeName=Base_hair_back_05_capsule, rotation=<rot 0 -90 0>}]
 Base_hair_side_04 [{shapeType=Capsule, length=0.505, radius=0.505, shapeName=Base_hair_side_04_capsule, rotation=<rot 0 -90 0>}]
 Base_hair_up_bk_06 [{shapeType=Capsule, length=0.505, radius=0.505, shapeName=Base_hair_up_bk_06_capsule, rotation=<rot 0 -90 0>}]
 Dyn_hair_L_01 [{
   shapeType=Capsule
   length=5.4993
   radius=1
   shapeName=Dyn_hair_L_01_capsule
   center=<5.7789 0.0562 -0.2438>
   rotation=<rot 77.6231 -0.0001 -5.5542>
  }]
 Dyn_hair_back_01 [{
   shapeType=Capsule
   length=5.1706
   radius=1.6976
   shapeName=Dyn_hair_back_01_capsule
   center=<5.6772 0.0304 -0.5493>
   rotation=<rot 78.5645 0 3.1621>
  }]
 Dyn_hair_back_02 [{
   shapeType=Capsule
   length=3.669
   radius=1.4778
   shapeName=Dyn_hair_back_02_capsule
   center=<4.5846 -0.0564 -0.3242>
   rotation=<rot 85.465 0.0003 3.3185>
  }]
 Dyn_hair_back_04 [{shapeType=Capsule, length=3.7619, radius=1.69, shapeName=Dyn_hair_back_04_capsule, center=<4.2164 0.0493 0>, rotation=<rot 90 0 0>}]
 Dyn_hair_back_05 [{
   shapeType=Capsule
   length=4.7222
   radius=1.6706
   shapeName=Dyn_hair_back_05_capsule
   center=<5.3357 0.0105 -1.3534>
   rotation=<rot 63.3286 0 -0.0001>
  }]
 Dyn_hair_front_01 [{
   shapeType=Capsule
   length=2.8077
   radius=1.7393
   shapeName=Dyn_hair_front_01_capsule
   center=<3.3475 0.2753 0>
   rotation=<rot 76.8514 2.698 0.0003>
  }]
 Dyn_hair_front_02 [{shapeType=Capsule, length=1.6412, radius=1.5855, shapeName=Dyn_hair_front_02_capsule, center=<1.7221 0 0>, rotation=<rot 90 0 0>}]
 Dyn_hair_side_01 [{
   shapeType=Capsule
   length=2.9024
   radius=1.6038
   shapeName=Dyn_hair_side_01_capsule
   center=<5.4299 0.2967 0.1991>
   rotation=<rot 90 0 -53.2679>
  }]
 Dyn_hair_side_02 [{
   shapeType=Capsule
   length=4.1113
   radius=0.9125
   shapeName=Dyn_hair_side_02_capsule
   center=<3.6444 0.0504 0.1787>
   rotation=<rot 59.8491 63.9619 61.4895>
  }]
 Dyn_hair_side_03 [{
   shapeType=Capsule
   length=4.6637
   radius=1.0679
   shapeName=Dyn_hair_side_03_capsule
   center=<3.5947 0 -0.028>
   rotation=<rot 86.7352 0.0005 -0.6896>
  }]
 Dyn_hair_side_04 [{
   shapeType=Capsule
   length=4.9723
   radius=1.6194
   shapeName=Dyn_hair_side_04_capsule
   center=<5.7885 -0.2596 -0.5307>
   rotation=<rot 77.7286 11.4244 13.6356>
  }]
 Dyn_hair_up_01 [{shapeType=Capsule, length=1.6412, radius=1.5855, shapeName=Dyn_hair_up_01_capsule, center=<1.7221 0 0>, rotation=<rot 90 0 0>}]
 Dyn_hair_up_02 [{shapeType=Capsule, length=1.6412, radius=1.5855, shapeName=Dyn_hair_up_02_capsule, center=<1.7221 0 0>, rotation=<rot 90 0 0>}]
 Dyn_hair_up_03 [{shapeType=Capsule, length=1.6412, radius=1.5855, shapeName=Dyn_hair_up_03_capsule, center=<1.7221 0 0>, rotation=<rot 90 0 0>}]
 Dyn_hair_up_04 [{shapeType=Capsule, length=1.8307, radius=1.396, shapeName=Dyn_hair_up_04_capsule, center=<2.8118 0 -0.3607>, rotation=<rot 75 0.0004 0.0005>}]
 Dyn_hair_up_06 [{shapeType=Capsule, length=1.6412, radius=1.5855, shapeName=Dyn_hair_up_06_capsule, center=<1.7221 0 0>, rotation=<rot 90 0 0>}]
 Dyn_hair_up_07 [{shapeType=Capsule, length=1.6412, radius=1.5855, shapeName=Dyn_hair_up_07_capsule, center=<1.7221 0 0>, rotation=<rot 90 0 0>}]
 Dyn_hair_up_L_08 [{shapeType=Capsule, length=1.6412, radius=1.5855, shapeName=Dyn_hair_up_L_08_capsule, center=<1.7221 0 0>, rotation=<rot 90 0 0>}]
 Dyn_hair_up_L_09 [{shapeType=Capsule, length=1.6412, radius=1.5855, shapeName=Dyn_hair_up_L_09_capsule, center=<1.7221 0 0>, rotation=<rot 90 0 0>}]
 Dyn_hair_up_bk_05 [{
   shapeType=Capsule
   length=3.619
   radius=1.3173
   shapeName=Dyn_hair_up_bk_05_capsule
   center=<3.8591 -0.2563 -1.369>
   rotation=<rot 0 -95.9269 -51.2307>
  }]
 Dyn_hair_up_bk_06 [{
   shapeType=Capsule
   length=4.0075
   radius=1.256
   shapeName=Dyn_hair_up_bk_06_capsule
   center=<3.9641 -0.0648 -1.1644>
   rotation=<rot 0 -92.3345 -53.4386>
  }]
 Master_hair_back [{
   shapeType=Capsule
   length=4.7405
   radius=2.3065
   shapeName=Master_hair_back_capsule
   center=<1.0923 1.1272 1.1353>
   rotation=<rot 0 -1.7022 -75.8755>
  }]
 base_hair_L_01 [{shapeType=Capsule, length=0.505, radius=0.505, shapeName=base_hair_L_01_capsule, rotation=<rot 0 -90 0>}]
 clavicle_l [{shapeType=Capsule, length=3.8805, radius=8.6379, shapeName=clavicle_l_capsule, center=<8.4694 0 0.741>, rotation=<rot -85 0.002 -0.002>}]
 clavicle_r [{shapeType=Capsule, length=3.2734, radius=9.245, shapeName=clavicle_l_capsule, center=<-8.4694 0 -0.741>, rotation=<rot -85 0.002 179.998>}]
 head [{shapeType=Capsule, length=5.9497, radius=5.7792, center=<1.9919 0 0>, rotation=<rot -90 -166.7373 166.7372>}]
 neck_01 [{shapeType=Capsule, length=8.409, radius=6.3455, shapeName=neck_01_capsule, rotation=<rot -90 -14.5688 14.5686>}]
 neck_02 [{shapeType=Capsule, length=6.5928, radius=5.8657, shapeName=neck_02_capsule, rotation=<rot -90 -12.5998 12.5996>}]
 spine_05 [{shapeType=Capsule, length=14.4529, radius=13.5082, shapeName=spine_05_capsule, center=<0 0.3241 0>, rotation=<rot 90 -179.9902 -179.99>}]
}
constraints [
 {
  bone1Name=head
  bone2Name=neck_02
  swing1LimitDegrees=45
  swing1Motion=Free
  swing2LimitDegrees=45
  swing2Motion=Free
  twistLimitDegrees=45
  twistMotion=Free
 }
 {
  bone1Name=neck_02
  bone2Name=neck_01
  swing1LimitDegrees=45
  swing1Motion=Free
  swing2LimitDegrees=45
  swing2Motion=Free
  twistLimitDegrees=45
  twistMotion=Free
 }
 {
  bone1Name=neck_01
  bone2Name=spine_05
  swing1LimitDegrees=45
  swing1Motion=Free
  swing2LimitDegrees=45
  swing2Motion=Free
  twistLimitDegrees=45
  twistMotion=Free
 }
 {
  bone1Name=Base_hair_side_04
  bone2Name=head
  swing1LimitDegrees=45
  swing1Motion=Free
  swing2LimitDegrees=45
  swing2Motion=Free
  twistLimitDegrees=45
  twistMotion=Free
 }
 {bone1Name=Dyn_hair_side_04, bone2Name=Base_hair_side_04, swing1LimitDegrees=7, swing2LimitDegrees=7, twistLimitDegrees=0}
 {bone1Name=Dyn_hair_front_01, bone2Name=head, swing1LimitDegrees=7, swing2LimitDegrees=7, twistLimitDegrees=0}
 {bone1Name=Dyn_hair_front_02, bone2Name=head, swing1LimitDegrees=7, swing2LimitDegrees=7, twistLimitDegrees=0}
 {bone1Name=Dyn_hair_side_01, bone2Name=head, swing1LimitDegrees=7, swing2LimitDegrees=7, twistLimitDegrees=0}
 {bone1Name=Dyn_hair_side_02, bone2Name=head, swing1LimitDegrees=7, swing2LimitDegrees=7, twistLimitDegrees=0}
 {bone1Name=Dyn_hair_side_03, bone2Name=head, swing1LimitDegrees=7, swing2LimitDegrees=7, twistLimitDegrees=0}
 {bone1Name=Dyn_hair_up_01, bone2Name=head, swing1LimitDegrees=7, swing2LimitDegrees=7, twistLimitDegrees=0}
 {bone1Name=Dyn_hair_up_02, bone2Name=head, swing1LimitDegrees=7, swing2LimitDegrees=7, twistLimitDegrees=0}
 {bone1Name=Dyn_hair_up_03, bone2Name=head, swing1LimitDegrees=7, swing2LimitDegrees=7, twistLimitDegrees=0}
 {bone1Name=Dyn_hair_up_04, bone2Name=head, swing1LimitDegrees=7, swing2LimitDegrees=7, twistLimitDegrees=0}
 {bone1Name=Dyn_hair_up_06, bone2Name=head, swing1LimitDegrees=7, swing2LimitDegrees=7, twistLimitDegrees=0}
 {bone1Name=Dyn_hair_up_07, bone2Name=head, swing1LimitDegrees=7, swing2LimitDegrees=7, twistLimitDegrees=0}
 {bone1Name=Dyn_hair_up_L_08, bone2Name=head, swing1LimitDegrees=7, swing2LimitDegrees=7, twistLimitDegrees=0}
 {bone1Name=Dyn_hair_up_L_09, bone2Name=head, swing1LimitDegrees=7, swing2LimitDegrees=7, twistLimitDegrees=0}
 {
  bone1Name=Master_hair_back
  bone2Name=head
  swing1LimitDegrees=45
  swing1Motion=Free
  swing2LimitDegrees=45
  swing2Motion=Free
  twistLimitDegrees=45
  twistMotion=Free
 }
 {
  bone1Name=Base_hair_back_01
  bone2Name=Master_hair_back
  swing1LimitDegrees=45
  swing1Motion=Free
  swing2LimitDegrees=45
  swing2Motion=Free
  twistLimitDegrees=45
  twistMotion=Free
 }
 {bone1Name=Dyn_hair_back_01, bone2Name=Base_hair_back_01, swing1LimitDegrees=7, swing2LimitDegrees=15, twistLimitDegrees=0}
 {
  bone1Name=Base_hair_back_02
  bone2Name=Master_hair_back
  swing1LimitDegrees=45
  swing1Motion=Free
  swing2LimitDegrees=45
  swing2Motion=Free
  twistLimitDegrees=45
  twistMotion=Free
 }
 {bone1Name=Dyn_hair_back_02, bone2Name=Base_hair_back_02, swing1LimitDegrees=10, swing2LimitDegrees=15, twistLimitDegrees=0}
 {
  bone1Name=Base_hair_back_04
  bone2Name=Master_hair_back
  swing1LimitDegrees=45
  swing1Motion=Free
  swing2LimitDegrees=45
  swing2Motion=Free
  twistLimitDegrees=45
  twistMotion=Free
 }
 {bone1Name=Dyn_hair_back_04, bone2Name=Base_hair_back_04, swing1LimitDegrees=10, swing2LimitDegrees=15, twistLimitDegrees=0}
 {
  bone1Name=Base_hair_back_05
  bone2Name=Master_hair_back
  swing1LimitDegrees=45
  swing1Motion=Free
  swing2LimitDegrees=45
  swing2Motion=Free
  twistLimitDegrees=45
  twistMotion=Free
 }
 {bone1Name=Dyn_hair_back_05, bone2Name=Base_hair_back_05, swing1LimitDegrees=5, swing2LimitDegrees=12, twistLimitDegrees=0}
 {
  bone1Name=base_hair_L_01
  bone2Name=Master_hair_back
  swing1LimitDegrees=45
  swing1Motion=Free
  swing2LimitDegrees=45
  swing2Motion=Free
  twistLimitDegrees=45
  twistMotion=Free
 }
 {bone1Name=Dyn_hair_L_01, bone2Name=base_hair_L_01, swing1LimitDegrees=12, swing2LimitDegrees=12, twistLimitDegrees=0}
 {
  bone1Name=clavicle_l
  bone2Name=spine_05
  swing1LimitDegrees=45
  swing1Motion=Free
  swing2LimitDegrees=45
  swing2Motion=Free
  twistLimitDegrees=45
  twistMotion=Free
 }
 {
  bone1Name=clavicle_r
  bone2Name=spine_05
  swing1LimitDegrees=45
  swing1Motion=Free
  swing2LimitDegrees=45
  swing2Motion=Free
  twistLimitDegrees=45
  twistMotion=Free
 }
 {bone1Name=Dyn_hair_up_bk_06, bone2Name=Base_hair_up_bk_06, swing1LimitDegrees=10, swing2LimitDegrees=10, twistLimitDegrees=0}
 {bone1Name=Dyn_hair_up_bk_05, bone2Name=head, swing1LimitDegrees=10, swing2LimitDegrees=10, twistLimitDegrees=0}
 {
  bone1Name=Base_hair_up_bk_06
  bone2Name=head
  swing1LimitDegrees=45
  swing1Motion=Free
  swing2LimitDegrees=45
  swing2Motion=Free
  twistLimitDegrees=45
  twistMotion=Free
 }
]
