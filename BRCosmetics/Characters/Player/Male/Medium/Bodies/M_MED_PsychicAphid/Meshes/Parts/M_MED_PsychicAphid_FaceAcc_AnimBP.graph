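# Post-physics anim blueprint for the PsychicAphid medium male face accessory, adding secondary motion to hair and earring.
# Copies the attached parent's pose into a Control Rig, nudges Master_hair_back, then runs left-earring AnimDynamics and a RigidBody simulation based on spine_05.
# Each update computes a gravity override through a deprecated helper and derives the rig's "RBAN off" input from bRigidBodyEnabled.
# A long chain of per-bone hair AnimDynamics nodes is disconnected from the output.
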AnimBlueprint /BRCosmetics/Characters/Player/Male/Medium/Bodies/M_MED_PsychicAphid/Meshes/Parts/M_MED_PsychicAphid_FaceAcc_AnimBP
target_skeleton=&/M_MED_PsychicAphid_FaceAcc_Skeleton
graphs [
 AnimGraph {nodes [
   n1 = Anim Dynamics - Bone: Dyn_hair_L_01(Alpha=1, ExternalForce="0,0,0") {
    angularDampingOverride=0.9999
    angularSpringConstant=300
    bAngularSpring=true
    bGravityOverrideInSimSpace=true
    bOverrideAngularDamping=true
    bUseGravityOverride=true
    alphaBoolBlend {blendOption=Linear}
    boundBone {boneName=Dyn_hair_L_01}
    componentAppliedLinearAccClamp=<1000 1000 1000>
    componentLinearAccScale=<3 3 3>
    componentLinearVelScale=<3 3 3>
    gravityOverride {z=-980}
    physicsBodyDefinitions [{
      boundBone {boneName=Dyn_hair_L_01}
      boxExtents=<10 10 10>
      constraintSetup {angularLimitsMax=<3 0 10>, angularLimitsMin=<-3 -15 -10>, angularTarget=<0 0 0>}
      localJointOffset=<0.5 0 0>
     }]
   }
   n2 = Anim Dynamics - Bone: Dyn_hair_front_01(ComponentPose=@n1, Alpha=1, ExternalForce="0,0,0") {
    angularDampingOverride=0.9999
    angularSpringConstant=300
    bAngularSpring=true
    bGravityOverrideInSimSpace=true
    bOverrideAngularDamping=true
    bUseGravityOverride=true
    alphaBoolBlend {blendOption=Linear}
    boundBone {boneName=Dyn_hair_front_01}
    componentAppliedLinearAccClamp=<1000 1000 1000>
    componentLinearAccScale=<3 3 3>
    componentLinearVelScale=<3 3 3>
    gravityOverride {z=-980}
    physicsBodyDefinitions [{
      boundBone {boneName=Dyn_hair_front_01}
      boxExtents=<10 10 10>
      constraintSetup {angularLimitsMax=<3 0 7>, angularLimitsMin=<-3 -15 -7>, angularTarget=<0 0 0>}
      localJointOffset=<0.5 0 0>
     }]
   }
   n3 = Anim Dynamics - Bone: Dyn_hair_front_02(ComponentPose=@n2, Alpha=1, ExternalForce="0,0,0") {
    angularDampingOverride=0.9999
    angularSpringConstant=300
    bAngularSpring=true
    bGravityOverrideInSimSpace=true
    bOverrideAngularDamping=true
    bUseGravityOverride=true
    alphaBoolBlend {blendOption=Linear}
    boundBone {boneName=Dyn_hair_front_02}
    componentAppliedLinearAccClamp=<1000 1000 1000>
    componentLinearAccScale=<3 3 3>
    componentLinearVelScale=<3 3 3>
    gravityOverride {z=-980}
    physicsBodyDefinitions [{
      boundBone {boneName=Dyn_hair_front_02}
      boxExtents=<10 10 10>
      constraintSetup {angularLimitsMax=<3 0 7>, angularLimitsMin=<-3 -15 -7>, angularTarget=<0 0 0>}
      localJointOffset=<0.5 0 0>
     }]
   }
   n4 = Anim Dynamics - Bone: Dyn_hair_side_01(ComponentPose=@n3, Alpha=1, ExternalForce="0,0,0") {
    angularDampingOverride=0.9999
    angularSpringConstant=300
    bAngularSpring=true
    bGravityOverrideInSimSpace=true
    bOverrideAngularDamping=true
    bUseGravityOverride=true
    alphaBoolBlend {blendOption=Linear}
    boundBone {boneName=Dyn_hair_side_01}
    componentAppliedLinearAccClamp=<1000 1000 1000>
    componentLinearAccScale=<3 3 3>
    componentLinearVelScale=<3 3 3>
    gravityOverride {z=-980}
    physicsBodyDefinitions [{
      boundBone {boneName=Dyn_hair_side_01}
      boxExtents=<10 10 10>
      constraintSetup {angularLimitsMax=<3 0 7>, angularLimitsMin=<-3 -15 -7>, angularTarget=<0 0 0>}
      localJointOffset=<0.5 0 0>
     }]
   }
   n5 = Anim Dynamics - Bone: Dyn_hair_side_02(ComponentPose=@n4, Alpha=1, ExternalForce="0,0,0") {
    angularDampingOverride=0.9999
    angularSpringConstant=300
    bAngularSpring=true
    bGravityOverrideInSimSpace=true
    bOverrideAngularDamping=true
    bUseGravityOverride=true
    alphaBoolBlend {blendOption=Linear}
    boundBone {boneName=Dyn_hair_side_02}
    componentAppliedLinearAccClamp=<1000 1000 1000>
    componentLinearAccScale=<3 3 3>
    componentLinearVelScale=<3 3 3>
    gravityOverride {z=-980}
    physicsBodyDefinitions [{
      boundBone {boneName=Dyn_hair_side_02}
      boxExtents=<10 10 10>
      constraintSetup {angularLimitsMax=<3 0 7>, angularLimitsMin=<-3 -15 -7>, angularTarget=<0 0 0>}
      localJointOffset=<0.5 0 0>
     }]
   }
   n6 = Anim Dynamics - Bone: Dyn_hair_side_03(ComponentPose=@n5, Alpha=1, ExternalForce="0,0,0") {
    angularDampingOverride=0.9999
    angularSpringConstant=300
    bAngularSpring=true
    bGravityOverrideInSimSpace=true
    bOverrideAngularDamping=true
    bUseGravityOverride=true
    alphaBoolBlend {blendOption=Linear}
    boundBone {boneName=Dyn_hair_side_03}
    componentAppliedLinearAccClamp=<1000 1000 1000>
    componentLinearAccScale=<3 3 3>
    componentLinearVelScale=<3 3 3>
    gravityOverride {z=-980}
    physicsBodyDefinitions [{
      boundBone {boneName=Dyn_hair_side_03}
      boxExtents=<10 10 10>
      constraintSetup {angularLimitsMax=<3 0 7>, angularLimitsMin=<-3 -15 -7>, angularTarget=<0 0 0>}
      localJointOffset=<0.5 0 0>
     }]
   }
   n7 = Anim Dynamics - Bone: Dyn_hair_side_04(ComponentPose=@n6, Alpha=1, ExternalForce="0,0,0") {
    angularDampingOverride=0.9999
    angularSpringConstant=300
    bAngularSpring=true
    bGravityOverrideInSimSpace=true
    bOverrideAngularDamping=true
    bUseGravityOverride=true
    alphaBoolBlend {blendOption=Linear}
    boundBone {boneName=Dyn_hair_side_04}
    componentAppliedLinearAccClamp=<1000 1000 1000>
    componentLinearAccScale=<3 3 3>
    componentLinearVelScale=<3 3 3>
    gravityOverride {z=-980}
    physicsBodyDefinitions [{
      boundBone {boneName=Dyn_hair_side_04}
      boxExtents=<10 10 10>
      constraintSetup {angularLimitsMax=<3 0 7>, angularLimitsMin=<-3 -15 -7>, angularTarget=<0 0 0>}
      localJointOffset=<0.5 0 0>
     }]
   }
   n8 = Anim Dynamics - Bone: Dyn_hair_side_05(ComponentPose=@n7, Alpha=1, ExternalForce="0,0,0") {
    angularDampingOverride=0.9999
    angularSpringConstant=300
    bAngularSpring=true
    bGravityOverrideInSimSpace=true
    bOverrideAngularDamping=true
    bUseGravityOverride=true
    alphaBoolBlend {blendOption=Linear}
    boundBone {boneName=Dyn_hair_side_05}
    componentAppliedLinearAccClamp=<1000 1000 1000>
    componentLinearAccScale=<3 3 3>
    componentLinearVelScale=<3 3 3>
    gravityOverride {z=-980}
    physicsBodyDefinitions [{
      boundBone {boneName=Dyn_hair_side_05}
      boxExtents=<10 10 10>
      constraintSetup {angularLimitsMax=<3 0 7>, angularLimitsMin=<-3 -15 -7>, angularTarget=<0 0 0>}
      localJointOffset=<0.5 0 0>
     }]
   }
   n9 = Anim Dynamics - Bone: Dyn_hair_up_01(ComponentPose=@n8, Alpha=1, ExternalForce="0,0,0") {
    angularDampingOverride=0.9999
    angularSpringConstant=300
    bAngularSpring=true
    bGravityOverrideInSimSpace=true
    bOverrideAngularDamping=true
    bUseGravityOverride=true
    alphaBoolBlend {blendOption=Linear}
    boundBone {boneName=Dyn_hair_up_01}
    componentAppliedLinearAccClamp=<1000 1000 1000>
    componentLinearAccScale=<3 3 3>
    componentLinearVelScale=<3 3 3>
    gravityOverride {z=-980}
    physicsBodyDefinitions [{
      boundBone {boneName=Dyn_hair_up_01}
      boxExtents=<10 10 10>
      constraintSetup {angularLimitsMax=<3 0 7>, angularLimitsMin=<-3 -15 -7>, angularTarget=<0 0 0>}
      localJointOffset=<0.5 0 0>
     }]
   }
   n10 = Anim Dynamics - Bone: Dyn_hair_up_02(ComponentPose=@n9, Alpha=1, ExternalForce="0,0,0") {
    angularDampingOverride=0.9999
    angularSpringConstant=300
    bAngularSpring=true
    bGravityOverrideInSimSpace=true
    bOverrideAngularDamping=true
    bUseGravityOverride=true
    alphaBoolBlend {blendOption=Linear}
    boundBone {boneName=Dyn_hair_up_02}
    componentAppliedLinearAccClamp=<1000 1000 1000>
    componentLinearAccScale=<3 3 3>
    componentLinearVelScale=<3 3 3>
    gravityOverride {z=-980}
    physicsBodyDefinitions [{
      boundBone {boneName=Dyn_hair_up_02}
      boxExtents=<10 10 10>
      constraintSetup {angularLimitsMax=<3 0 7>, angularLimitsMin=<-3 -15 -7>, angularTarget=<0 0 0>}
      localJointOffset=<0.5 0 0>
     }]
   }
   n11 = Anim Dynamics - Bone: Dyn_hair_up_03(ComponentPose=@n10, Alpha=1, ExternalForce="0,0,0") {
    angularDampingOverride=0.9999
    angularSpringConstant=300
    bAngularSpring=true
    bGravityOverrideInSimSpace=true
    bOverrideAngularDamping=true
    bUseGravityOverride=true
    alphaBoolBlend {blendOption=Linear}
    boundBone {boneName=Dyn_hair_up_03}
    componentAppliedLinearAccClamp=<1000 1000 1000>
    componentLinearAccScale=<3 3 3>
    componentLinearVelScale=<3 3 3>
    gravityOverride {z=-980}
    physicsBodyDefinitions [{
      boundBone {boneName=Dyn_hair_up_03}
      boxExtents=<10 10 10>
      constraintSetup {angularLimitsMax=<3 0 7>, angularLimitsMin=<-3 -15 -7>, angularTarget=<0 0 0>}
      localJointOffset=<0.5 0 0>
     }]
   }
   n12 = Anim Dynamics - Bone: Dyn_hair_up_04(ComponentPose=@n11, Alpha=1, ExternalForce="0,0,0") {
    angularDampingOverride=0.9999
    angularSpringConstant=300
    bAngularSpring=true
    bGravityOverrideInSimSpace=true
    bOverrideAngularDamping=true
    bUseGravityOverride=true
    alphaBoolBlend {blendOption=Linear}
    boundBone {boneName=Dyn_hair_up_04}
    componentAppliedLinearAccClamp=<1000 1000 1000>
    componentLinearAccScale=<3 3 3>
    componentLinearVelScale=<3 3 3>
    gravityOverride {z=-980}
    physicsBodyDefinitions [{
      boundBone {boneName=Dyn_hair_up_04}
      boxExtents=<10 10 10>
      constraintSetup {angularLimitsMax=<3 0 7>, angularLimitsMin=<-3 -15 -7>, angularTarget=<0 0 0>}
      localJointOffset=<0.5 0 0>
     }]
   }
   n13 = Anim Dynamics - Bone: Dyn_hair_up_bk_05(ComponentPose=@n12, Alpha=1, ExternalForce="0,0,0") {
    angularDampingOverride=0.9999
    angularSpringConstant=300
    bAngularSpring=true
    bGravityOverrideInSimSpace=true
    bOverrideAngularDamping=true
    bUseGravityOverride=true
    alphaBoolBlend {blendOption=Linear}
    boundBone {boneName=Dyn_hair_up_bk_05}
    componentAppliedLinearAccClamp=<1000 1000 1000>
    componentLinearAccScale=<3 3 3>
    componentLinearVelScale=<3 3 3>
    gravityOverride {z=-980}
    physicsBodyDefinitions [{
      boundBone {boneName=Dyn_hair_up_bk_05}
      boxExtents=<10 10 10>
      constraintSetup {angularLimitsMax=<3 0 7>, angularLimitsMin=<-3 -15 -7>, angularTarget=<0 0 0>}
      localJointOffset=<0.5 0 0>
     }]
   }
   n14 = Anim Dynamics - Bone: Dyn_hair_up_06(ComponentPose=@n13, Alpha=1, ExternalForce="0,0,0") {
    angularDampingOverride=0.9999
    angularSpringConstant=300
    bAngularSpring=true
    bGravityOverrideInSimSpace=true
    bOverrideAngularDamping=true
    bUseGravityOverride=true
    alphaBoolBlend {blendOption=Linear}
    boundBone {boneName=Dyn_hair_up_06}
    componentAppliedLinearAccClamp=<1000 1000 1000>
    componentLinearAccScale=<3 3 3>
    componentLinearVelScale=<3 3 3>
    gravityOverride {z=-980}
    physicsBodyDefinitions [{
      boundBone {boneName=Dyn_hair_up_06}
      boxExtents=<10 10 10>
      constraintSetup {angularLimitsMax=<3 0 7>, angularLimitsMin=<-3 -15 -7>, angularTarget=<0 0 0>}
      localJointOffset=<0.5 0 0>
     }]
   }
   n15 = Anim Dynamics - Bone: Dyn_hair_up_07(ComponentPose=@n14, Alpha=1, ExternalForce="0,0,0") {
    angularDampingOverride=0.9999
    angularSpringConstant=300
    bAngularSpring=true
    bGravityOverrideInSimSpace=true
    bOverrideAngularDamping=true
    bUseGravityOverride=true
    alphaBoolBlend {blendOption=Linear}
    boundBone {boneName=Dyn_hair_up_07}
    componentAppliedLinearAccClamp=<1000 1000 1000>
    componentLinearAccScale=<3 3 3>
    componentLinearVelScale=<3 3 3>
    gravityOverride {z=-980}
    physicsBodyDefinitions [{
      boundBone {boneName=Dyn_hair_up_07}
      boxExtents=<10 10 10>
      constraintSetup {angularLimitsMax=<3 0 7>, angularLimitsMin=<-3 -15 -7>, angularTarget=<0 0 0>}
      localJointOffset=<0.5 0 0>
     }]
   }
   n16 = Anim Dynamics - Bone: Dyn_hair_up_L_08(ComponentPose=@n15, Alpha=1, ExternalForce="0,0,0") {
    angularDampingOverride=0.9999
    angularSpringConstant=300
    bAngularSpring=true
    bGravityOverrideInSimSpace=true
    bOverrideAngularDamping=true
    bUseGravityOverride=true
    alphaBoolBlend {blendOption=Linear}
    boundBone {boneName=Dyn_hair_up_L_08}
    componentAppliedLinearAccClamp=<1000 1000 1000>
    componentLinearAccScale=<3 3 3>
    componentLinearVelScale=<3 3 3>
    gravityOverride {z=-980}
    physicsBodyDefinitions [{
      boundBone {boneName=Dyn_hair_up_L_08}
      boxExtents=<10 10 10>
      constraintSetup {angularLimitsMax=<3 0 7>, angularLimitsMin=<-3 -15 -7>, angularTarget=<0 0 0>}
      localJointOffset=<0.5 0 0>
     }]
   }
   n17 = Anim Dynamics - Bone: Dyn_hair_up_L_09(ComponentPose=@n16, Alpha=1, ExternalForce="0,0,0") {
    angularDampingOverride=0.9999
    angularSpringConstant=300
    bAngularSpring=true
    bGravityOverrideInSimSpace=true
    bOverrideAngularDamping=true
    bUseGravityOverride=true
    alphaBoolBlend {blendOption=Linear}
    boundBone {boneName=Dyn_hair_up_L_09}
    componentAppliedLinearAccClamp=<1000 1000 1000>
    componentLinearAccScale=<3 3 3>
    componentLinearVelScale=<3 3 3>
    gravityOverride {z=-980}
    physicsBodyDefinitions [{
      boundBone {boneName=Dyn_hair_up_L_09}
      boxExtents=<10 10 10>
      constraintSetup {angularLimitsMax=<3 0 7>, angularLimitsMin=<-3 -15 -7>, angularTarget=<0 0 0>}
      localJointOffset=<0.5 0 0>
     }]
   }
   n18 = Anim Dynamics - Bone: Dyn_hair_back_01(ComponentPose=@n17, Alpha=1, ExternalForce="0,0,0") {
    angularDampingOverride=0.9999
    angularSpringConstant=300
    bAngularSpring=true
    bGravityOverrideInSimSpace=true
    bOverrideAngularDamping=true
    bUseGravityOverride=true
    alphaBoolBlend {blendOption=Linear}
    boundBone {boneName=Dyn_hair_back_01}
    componentAppliedLinearAccClamp=<1000 1000 1000>
    componentLinearAccScale=<3 3 3>
    componentLinearVelScale=<3 3 3>
    gravityOverride {z=-980}
    physicsBodyDefinitions [{
      boundBone {boneName=Dyn_hair_back_01}
      boxExtents=<10 10 10>
      constraintSetup {angularLimitsMax=<3 0 10>, angularLimitsMin=<-3 -15 -10>, angularTarget=<0 0 0>}
      localJointOffset=<0.5 0 0>
     }]
   }
   n19 = Anim Dynamics - Bone: Dyn_hair_back_02(ComponentPose=@n18, Alpha=1, ExternalForce="0,0,0") {
    angularDampingOverride=0.9999
    angularSpringConstant=300
    bAngularSpring=true
    bGravityOverrideInSimSpace=true
    bOverrideAngularDamping=true
    bUseGravityOverride=true
    alphaBoolBlend {blendOption=Linear}
    boundBone {boneName=Dyn_hair_back_02}
    componentAppliedLinearAccClamp=<1000 1000 1000>
    componentLinearAccScale=<3 3 3>
    componentLinearVelScale=<3 3 3>
    gravityOverride {z=-980}
    physicsBodyDefinitions [{
      boundBone {boneName=Dyn_hair_back_02}
      boxExtents=<10 10 10>
      constraintSetup {angularLimitsMax=<3 0 10>, angularLimitsMin=<-3 -15 -10>, angularTarget=<0 0 0>}
      localJointOffset=<0.5 0 0>
     }]
   }
   n20 = Anim Dynamics - Bone: Dyn_hair_back_03(ComponentPose=@n19, Alpha=1, ExternalForce="0,0,0") {
    angularDampingOverride=0.9999
    angularSpringConstant=300
    bAngularSpring=true
    bGravityOverrideInSimSpace=true
    bOverrideAngularDamping=true
    bUseGravityOverride=true
    alphaBoolBlend {blendOption=Linear}
    boundBone {boneName=Dyn_hair_back_03}
    componentAppliedLinearAccClamp=<1000 1000 1000>
    componentLinearAccScale=<3 3 3>
    componentLinearVelScale=<3 3 3>
    gravityOverride {z=-980}
    physicsBodyDefinitions [{
      boundBone {boneName=Dyn_hair_back_03}
      boxExtents=<10 10 10>
      constraintSetup {angularLimitsMax=<3 0 7>, angularLimitsMin=<-3 -10 -7>, angularTarget=<0 0 0>}
      localJointOffset=<0.5 0 0>
     }]
   }
   n21 = Anim Dynamics - Bone: Dyn_hair_back_04(ComponentPose=@n20, Alpha=1, ExternalForce="0,0,0") {
    angularDampingOverride=0.9999
    angularSpringConstant=300
    bAngularSpring=true
    bGravityOverrideInSimSpace=true
    bOverrideAngularDamping=true
    bUseGravityOverride=true
    alphaBoolBlend {blendOption=Linear}
    boundBone {boneName=Dyn_hair_back_04}
    componentAppliedLinearAccClamp=<1000 1000 1000>
    componentLinearAccScale=<3 3 3>
    componentLinearVelScale=<3 3 3>
    gravityOverride {z=-980}
    physicsBodyDefinitions [{
      boundBone {boneName=Dyn_hair_back_04}
      boxExtents=<10 10 10>
      constraintSetup {angularLimitsMax=<3 0 10>, angularLimitsMin=<-3 -15 -10>, angularTarget=<0 0 0>}
      localJointOffset=<0.5 0 0>
     }]
   }
   n22 = Anim Dynamics - Bone: Dyn_hair_back_05(ComponentPose=@n21, Alpha=1, ExternalForce="0,0,0") {
    angularDampingOverride=0.9999
    angularSpringConstant=300
    bAngularSpring=true
    bGravityOverrideInSimSpace=true
    bOverrideAngularDamping=true
    bUseGravityOverride=true
    alphaBoolBlend {blendOption=Linear}
    boundBone {boneName=Dyn_hair_back_05}
    componentAppliedLinearAccClamp=<1000 1000 1000>
    componentLinearAccScale=<3 3 3>
    componentLinearVelScale=<3 3 3>
    gravityOverride {z=-980}
    physicsBodyDefinitions [{
      boundBone {boneName=Dyn_hair_back_05}
      boxExtents=<10 10 10>
      constraintSetup {angularLimitsMax=<3 0 10>, angularLimitsMin=<-3 -15 -10>, angularTarget=<0 0 0>}
      localJointOffset=<0.5 0 0>
     }]
   }
   Anim Dynamics - Bone: Dyn_hair_back_06(ComponentPose=@n22, Alpha=1, ExternalForce="0,0,0") {
    angularDampingOverride=0.9999
    angularSpringConstant=300
    bAngularSpring=true
    bGravityOverrideInSimSpace=true
    bOverrideAngularDamping=true
    bUseGravityOverride=true
    alphaBoolBlend {blendOption=Linear}
    boundBone {boneName=Dyn_hair_back_06}
    componentAppliedLinearAccClamp=<1000 1000 1000>
    componentLinearAccScale=<3 3 3>
    componentLinearVelScale=<3 3 3>
    gravityOverride {z=-980}
    physicsBodyDefinitions [{
      boundBone {boneName=Dyn_hair_back_06}
      boxExtents=<10 10 10>
      constraintSetup {angularLimitsMax=<3 0 7>, angularLimitsMin=<-3 -10 -7>, angularTarget=<0 0 0>}
      localJointOffset=<0.5 0 0>
     }]
   }
   n23 = Control Rig!(Source=Copy Pose From Mesh() {bUseAttachedParent=true}, Alpha, RBAN off:bool=Get RBAN off() -> (RBAN off:bool)) -> (Pose) {
    alphaBoolBlend {blendOption=Linear}
    controlRigAssetReference {blueprintRigClass=&/M_MED_PsychicAphid_FaceAcc_CtrlRig.M_MED_PsychicAphid_FaceAcc_CtrlRig_C}
   }
   n24 = "Transform (Modify) Bone - Bone: Master_hair_back"(ComponentPose=Local To Component(LocalPose=@n23), Alpha=1, Translation="0,-0.5,0.5", Rotation="0,0,5", Scale="1,1,1") {
    rotationMode=BMM_Additive
    scaleMode=BMM_Additive
    translationMode=BMM_Additive
    alphaBoolBlend {blendOption=Linear}
    boneToModify {boneName=Master_hair_back}
    rotation {roll=5}
    translation {y=-0.5, z=0.5}
   }
   n25 = Anim Dynamics - Bone: Dyn_earring_L!(ComponentPose=@n24, Alpha=1, GravityOverride:Vector="0,0,-980", ExternalForce="0,0,0") -> (Pose) {
    angularDampingOverride=0.999
    angularSpringConstant=500
    bAngularSpring=true
    bOverrideAngularDamping=true
    bUseGravityOverride=true
    simulationSpace=RootRelative
    alphaBoolBlend {blendOption=Linear}
    boundBone {boneName=Dyn_earring_L}
    componentAppliedLinearAccClamp=<10000 10000 10000>
    componentLinearAccScale=<5 5 5>
    componentLinearVelScale=<5 5 5>
    gravityOverride {z=-980}
    physicsBodyDefinitions [{
      boundBone {boneName=Dyn_earring_L}
      boxExtents=<7 7 7>
      constraintSetup {angularTargetAxis=AxisZ, angularLimitsMax=<0 5 30>, angularLimitsMin=<0 -15 -30>, angularTarget=<0 0 1>}
      localJointOffset=<0.6 0 0>
     }]
   }
   n26 = RigidBody!(ComponentPose=@n25, Alpha, OverrideWorldGravity:Vector="0,0,-980", ExternalForce) -> (Pose) {
    bClampLinearTranslationLimitToRefPose=true
    bOverrideWorldGravity=true
    overridePhysicsAsset=&/M_MED_PsychicAphid_FaceAcc_Physics
    simulationSpace=BaseBoneSpace
    alphaBoolBlend {blendOption=Linear}
    baseBoneRef {boneName=spine_05}
    componentAppliedLinearAccClamp=<2000 2000 2000>
    overrideWorldGravity {z=-980}
    simSpaceSettings {
     maxAngularAcceleration=500
     maxAngularVelocity=3
     maxLinearAcceleration=400
     maxLinearVelocity=250
     velocityScaleZ=0.5
     worldAlpha=0.5
     externalLinearDragV=<0.1 0.1>
    }
   }
   Output Pose(Result=Component To Local(ComponentPose=@n26)) {layerGroup=None}
  ]}
 EventGraph {nodes [
   Event BlueprintUpdateAnimation()
   > n1 = Gravity Override State Function DEPRECATED(
    CharacterPartAnimInstance=Self-Reference()
    GameTimeInSeconds=GetGameTimeInSeconds()
    GravityOverrideParams=Get Gravity Override Params()
    DebugJointName=dyn_hackBottomArmor
   )
   > Set Override World Gravity(Override World Gravity:Vector=@n1) -> (Output_Get:Vector)
   > Set RBAN off(RBAN off:bool=NOT Boolean(A=Get bRigidBodyEnabled())) -> (Output_Get:bool)
  ]}
]
variables [
 Gravity Override Params {type=GravityOverrideParamsStruct}
 Override World Gravity {type=Vector}
 RBAN off {type=boolean}
]
